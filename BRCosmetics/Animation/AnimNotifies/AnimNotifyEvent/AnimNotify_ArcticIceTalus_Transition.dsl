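; Anim notify state that drives cosmetic material effects on a player pawn's Backpack and Face meshes during the ArcticIce Talus transition.
; Each tick, parts with "ArcticIce" in their name get overlay outline and mesh-scale overrides, and the Backpack also copies its MaterialCurveA animation curve to the overlay.
; Other parts get MaterialCurveB set on their materials instead.
; Notify end restores outline visibility and resets mesh scale on ArcticIce parts only.
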(fn Received_NotifyEnd
  (params (MeshComp :type Skeletal%20Mesh%20Component%20Object%20Reference) (Animation :type Anim%20Sequence%20Base%20Object%20Reference) (EventReference :type Anim%20Notify%20Event%20Reference%20Structure%20%28by%20ref%29))
  (bind _asfort_player_pawn (Utilities|Casting|CastToFortPlayerPawn (Components|GetOwner MeshComp)))
  (Utilities|FlowControl|Sequence
    (:then_0
      (if (> (Utilities|String|FindSubstring (Utilities|GetDisplayName (Customization|GetSkeletalMeshforPartType _asfort_player_pawn "Backpack" :!class /Script/FortniteGame.FortPlayerPawn)) "ArcticIce" true) -1)
        (Class|FortCosmeticMaterialBlueprintLibrary|SetScalarParameterValueonOverlayMaterial (Customization|GetSkeletalMeshforPartType _asfort_player_pawn "Backpack" :!class /Script/FortniteGame.FortPlayerPawn) "OutlineVisibility" 1.0)
        (Class|FortCosmeticMaterialBlueprintLibrary|SetScalarParameterValueonOverlayMaterial (Customization|GetSkeletalMeshforPartType _asfort_player_pawn "Backpack" :!class /Script/FortniteGame.FortPlayerPawn) "ScaleMesh")))
    (:then_1
      (if (> (Utilities|String|FindSubstring (Utilities|GetDisplayName (Customization|GetSkeletalMeshforPartType _asfort_player_pawn "Face" :!class /Script/FortniteGame.FortPlayerPawn)) "ArcticIce" true) -1)
        (Class|FortCosmeticMaterialBlueprintLibrary|SetScalarParameterValueonOverlayMaterial (Customization|GetSkeletalMeshforPartType _asfort_player_pawn "Face" :!class /Script/FortniteGame.FortPlayerPawn) "OutlineVisibility" 1.0)
        (Class|FortCosmeticMaterialBlueprintLibrary|SetScalarParameterValueonOverlayMaterial (Customization|GetSkeletalMeshforPartType _asfort_player_pawn "Face" :!class /Script/FortniteGame.FortPlayerPawn) "ScaleMesh")
        (return false)))))

(fn Received_NotifyTick
  (params (MeshComp :type Skeletal%20Mesh%20Component%20Object%20Reference) (Animation :type Anim%20Sequence%20Base%20Object%20Reference) (FrameDeltaTime :type Float%20%28single-precision%29) (EventReference :type Anim%20Notify%20Event%20Reference%20Structure%20%28by%20ref%29))
  (bind _asfort_player_pawn (Utilities|Casting|CastToFortPlayerPawn (Components|GetOwner MeshComp)))
  (Utilities|FlowControl|Sequence
    (:then_0
      (if (> (Utilities|String|FindSubstring (Utilities|GetDisplayName (Customization|GetSkeletalMeshforPartType _asfort_player_pawn "Backpack" :!class /Script/FortniteGame.FortPlayerPawn)) "ArcticIce" true) -1)
        (Class|FortCosmeticMaterialBlueprintLibrary|SetScalarParameterValueonOverlayMaterial (Customization|GetSkeletalMeshforPartType _asfort_player_pawn "Backpack" :!class /Script/FortniteGame.FortPlayerPawn) "OutlineVisibility")
        (Class|FortCosmeticMaterialBlueprintLibrary|SetScalarParameterValueonOverlayMaterial (Customization|GetSkeletalMeshforPartType _asfort_player_pawn "Backpack" :!class /Script/FortniteGame.FortPlayerPawn) "ScaleMesh" 5.0)
        (Class|FortCosmeticMaterialBlueprintLibrary|SetScalarParameterValueonOverlayMaterial (Customization|GetSkeletalMeshforPartType _asfort_player_pawn "Backpack" :!class /Script/FortniteGame.FortPlayerPawn) "MaterialCurveA" (Animation|Curves|GetCurveValue (Components|SkeletalMesh|GetAnimInstance (Customization|GetSkeletalMeshforPartType _asfort_player_pawn "Backpack" :!class /Script/FortniteGame.FortPlayerPawn)) "MaterialCurveA"))
        (else
          (Rendering|Material|SetScalarParameterValueonMaterials (Customization|GetSkeletalMeshforPartType _asfort_player_pawn "Backpack" :!class /Script/FortniteGame.FortPlayerPawn) "MaterialCurveB" 1.0))))
    (:then_1
      (if (> (Utilities|String|FindSubstring (Utilities|GetDisplayName (Customization|GetSkeletalMeshforPartType _asfort_player_pawn "Face" :!class /Script/FortniteGame.FortPlayerPawn)) "ArcticIce" true) -1)
        (Class|FortCosmeticMaterialBlueprintLibrary|SetScalarParameterValueonOverlayMaterial (Customization|GetSkeletalMeshforPartType _asfort_player_pawn "Face" :!class /Script/FortniteGame.FortPlayerPawn) "OutlineVisibility")
        (Class|FortCosmeticMaterialBlueprintLibrary|SetScalarParameterValueonOverlayMaterial (Customization|GetSkeletalMeshforPartType _asfort_player_pawn "Face" :!class /Script/FortniteGame.FortPlayerPawn) "ScaleMesh" 3.0)
        (return false)
        (else
          (Rendering|Material|SetScalarParameterValueonMaterials (Customization|GetSkeletalMeshforPartType _asfort_player_pawn "Face" :!class /Script/FortniteGame.FortPlayerPawn) "MaterialCurveB" 1.0))))))
