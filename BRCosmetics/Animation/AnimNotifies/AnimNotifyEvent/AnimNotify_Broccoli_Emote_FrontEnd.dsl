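; Front-end emote anim notify that toggles the Broccoli cosmetic character on the owning player pawn between human and gorilla (monkey) forms, preserving casual style.
; On notify, it caches the pawn's cosmetic meta tags, then sets all four swap IDs to Inactive.
; The cached Monkey and Casual tags then pick the opposite form's swap to set Active.
; String-hash and pedestal-pawn helper functions are defined but not called here.
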(fn GetStringHashValue
  (params (Input :type String%20%28by%20ref%29))
  (for _index (range (Utilities|String|Len Input))
    (LocalVariables|SetAccumulator (+ (LocalVariables|GetAccumulator) (Utilities|String|GetCharacterasNumber Input _index))))
  (return (Math|Integer|%25%28Integer%29 (LocalVariables|GetAccumulator) 257)))

(fn IsPedestalPawn
  (params (Pawn :type Fort%20Player%20Pawn%20Object%20Reference%20%28by%20ref%29) (Pedestals :type Array%20of%20Fort%20Team%20Member%20Pedestal%20Object%20References))
  (for _array_element Pedestals
    (if (Utilities|Equal(Object) (Variables|/Script/FortniteUI.FortItemPlacementActor|GetCurrentPawn _array_element) Pawn)
      (return true)))
  (return false))

(fn Received_Notify
  (params (MeshComp :type Skeletal%20Mesh%20Component%20Object%20Reference) (Animation :type Anim%20Sequence%20Base%20Object%20Reference) (EventReference :type Anim%20Notify%20Event%20Reference%20Structure%20%28by%20ref%29))
  (bind _cachedmetatags (LocalVariables|GetCachedMetaTags))
  (bind _asfort_player_pawn (Utilities|Casting|CastToFortPlayerPawn (Components|GetOwner MeshComp)))
  (LocalVariables|SetPlayerPawn _asfort_player_pawn)
  (LocalVariables|SetCachedMetaTags (Cosmetics|GatherMetaTags _asfort_player_pawn "/BRCosmetics/Athena/Items/Cosmetics/Characters/CID_A_092_Athena_Commando_M_Broccoli_PR297.CID_A_092_Athena_Commando_M_Broccoli_PR297" :!class /Script/FortniteGame.FortPlayerPawn))
  (Utilities|FlowControl|Sequence
    (:then_0
      (for _array_element (Utilities|Array|MakeArray (Variables|Default|GetGorillaSwapId) (Variables|Default|GetHumanSwapId) (Variables|Default|GetCasualSwapId) (Variables|Default|GetGorillaCasualSwapId))
        (Class|FortPlayerPawn|SwapPawnCosmeticItemandVariant (LocalVariables|GetPlayerPawn) _array_element (Utilities|Array|MakeArray) true "Inactive" false)))
    (:then_1
      (LocalVariables|SetisCasual (GameplayTags|HasAnyTags _cachedmetatags (GameplayTags|MakeLiteralGameplayTagContainer "(GameplayTags=((TagName=\"Cosmetics.Variant.Meta.Broccoli.Casual\"),(TagName=\"Cosmetics.Variant.Meta.Broccoli.Monkey.Casual\")))")))
      (if (GameplayTags|HasTag _cachedmetatags "(TagName=\"Cosmetics.Variant.Meta.Broccoli.Monkey\")")
        (if (LocalVariables|GetisCasual)
          (Class|FortPlayerPawn|SwapPawnCosmeticItemandVariant (LocalVariables|GetPlayerPawn) (Variables|Default|GetCasualSwapId) (Variables|Default|GetCasualSwapData) true "Active" false)
          (else
            (Class|FortPlayerPawn|SwapPawnCosmeticItemandVariant (LocalVariables|GetPlayerPawn) (Variables|Default|GetHumanSwapId) (Variables|Default|GetHumanSwapData) true "Active" false)))
        (elif (LocalVariables|GetisCasual)
          (Class|FortPlayerPawn|SwapPawnCosmeticItemandVariant (LocalVariables|GetPlayerPawn) (Variables|Default|GetGorillaCasualSwapId) (Variables|Default|GetGorillaCasualSwapData) true "Active" false)
          (else
            (Class|FortPlayerPawn|SwapPawnCosmeticItemandVariant (LocalVariables|GetPlayerPawn) (Variables|Default|GetGorillaSwapId) (Variables|Default|GetGorillaSwapData) true "Active" false))))
      (return false))))
