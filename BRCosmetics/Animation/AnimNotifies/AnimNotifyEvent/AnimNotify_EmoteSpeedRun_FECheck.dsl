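; Anim notify for the SpeedRun emote that spawns speed-line particles on the mesh's Root, choosing a variant by player controller presence.
; It looks up the Fort player controller from the mesh component's owner.
; A valid controller gets the standard lines emitter.
; Without one, it prints a debug string and attaches the FE variant (likely front end) at a different offset.
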(fn Received_Notify
  (params (MeshComp :type Skeletal%20Mesh%20Component%20Object%20Reference) (Animation :type Anim%20Sequence%20Base%20Object%20Reference) (EventReference :type Anim%20Notify%20Event%20Reference%20Structure%20%28by%20ref%29))
  (Utilities|IsValid (Fortnite|GetFortPlayerControllerfromActor (Components|GetOwner MeshComp))
    (:Is%20Valid
      (Effects|SpawnEmitterAttached "/BRCosmetics/Athena/Effects/Particles/Emotes/P_Emote_SpeedRun_Lines.P_Emote_SpeedRun_Lines" MeshComp "Root" "0.000000,30.000000,90.000000" "0,-90.000000, 0"))
    (:Is%20Not%20Valid
      (Development|PrintString "No Player Controller")
      (Effects|SpawnEmitterAttached "/BRCosmetics/Athena/Effects/Particles/Emotes/P_Emote_SpeedRun_Lines_FE.P_Emote_SpeedRun_Lines_FE" MeshComp "Root" "0, 0,90.000000" "0,-90.000000, 0")))
  (return false))
