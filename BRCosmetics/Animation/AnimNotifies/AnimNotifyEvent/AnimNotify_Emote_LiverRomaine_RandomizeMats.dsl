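; Anim notify for the LiverRomaine emote that randomizes materials on the skeletal mesh playing the animation.
; When the notify fires, it loops over a configurable array of material slot indexes.
; For each slot, it sets a material picked at random from a configurable materials array, so slots roll independently and can share a material.
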(fn Received_Notify
  (params (MeshComp :type Skeletal%20Mesh%20Component%20Object%20Reference) (Animation :type Anim%20Sequence%20Base%20Object%20Reference) (EventReference :type Anim%20Notify%20Event%20Reference%20Structure%20%28by%20ref%29))
  (for _array_element (Variables|Default|GetMaterialIndexes)
    (Rendering|Material|SetMaterial (Variables|GetMeshComp) _array_element (Utilities|Array|Get(acopy) (Variables|Default|GetMaterials) (Math|Random|RandomIntegerinRange 0 (Utilities|Array|LastIndex (Variables|Default|GetMaterials))))))
  (return true))
