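; Anim notify state that, going by its name, takes the glasses off the Emperor cosmetic during an animation.
; Each notify tick casts the mesh owner to FortPlayerPawn and fetches its Face part skeletal mesh.
; It sets the "Hide Element 01" and "Use Custom Opacity" material scalars to on, with no reset in this graph.
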(fn Received_NotifyTick
  (params (MeshComp :type Skeletal%20Mesh%20Component%20Object%20Reference) (Animation :type Anim%20Sequence%20Base%20Object%20Reference) (FrameDeltaTime :type Float%20%28single-precision%29) (EventReference :type Anim%20Notify%20Event%20Reference%20Structure%20%28by%20ref%29))
  (bind _asfort_player_pawn (Utilities|Casting|CastToFortPlayerPawn (Components|GetOwner MeshComp)))
  (Rendering|Material|SetScalarParameterValueonMaterials (Customization|GetSkeletalMeshforPartType _asfort_player_pawn "Face" :!class /Script/FortniteGame.FortPlayerPawn) "Hide Element 01" 1.0)
  (Rendering|Material|SetScalarParameterValueonMaterials (Customization|GetSkeletalMeshforPartType _asfort_player_pawn "Face" :!class /Script/FortniteGame.FortPlayerPawn) "Use Custom Opacity" 1.0)
  (return false))
