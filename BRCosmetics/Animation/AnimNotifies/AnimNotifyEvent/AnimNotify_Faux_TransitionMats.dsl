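; Anim notify state that swaps the Faux character's materials to "unzip" alien transition versions during an animation.
; It acts only when the pawn's head mesh is the female Faux head, using the active Material variant channel to pick default or Cream materials.
; Begin applies dynamic material instances to Face, Head and Body.
; End, if Reset Materials When Done is set, reapplies hardcoded standard materials and prints a debug string.
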(fn Received_NotifyBegin
  (params (MeshComp :type Skeletal%20Mesh%20Component%20Object%20Reference) (Animation :type Anim%20Sequence%20Base%20Object%20Reference) (TotalDuration :type Float%20%28single-precision%29) (EventReference :type Anim%20Notify%20Event%20Reference%20Structure%20%28by%20ref%29))
  (bind _asfort_player_pawn (Utilities|Casting|CastToFortPlayerPawn (Components|GetOwner MeshComp)))
  (Utilities|FlowControl|Sequence
    (:then_0
      (Utilities|IsValid (Customization|GetSkeletalMeshforPartType _asfort_player_pawn "Head" :!class /Script/FortniteGame.FortPlayerPawn)
        (:Is%20Valid
          (if (Utilities|Equal(Object) "/BRCosmetics/Characters/Player/Female/Medium/Heads/F_MED_Faux_Head/Meshes/Female_Medium_Faux_Head.Female_Medium_Faux_Head" (Variables|/Script/Engine.SkinnedMeshComponent|GetSkeletalMesh (Customization|GetSkeletalMeshforPartType _asfort_player_pawn "Head" :!class /Script/FortniteGame.FortPlayerPawn)))
            (LocalVariables|SetCurrent%20Progressive%20Variant 1)))))
    (:then_1
      (bind _index (CosmeticLoadout|GetVariantChannelActiveIndexforSlotActor _asfort_player_pawn "(TagName=\"LoadoutSlot.Character\")" _asfort_player_pawn "(TagName=\"Cosmetics.Variant.Channel.Material\")" (Utilities|Array|MakeArray "(TagName=\"Cosmetics.Variant.Property.Mat1\")" "(TagName=\"Cosmetics.Variant.Property.Mat2\")")))
      (LocalVariables|SetCurrent%20Material%20Variant _index))
    (:then_2
      (if (== (LocalVariables|GetCurrent%20Progressive%20Variant) 1)
        (CallFunction|SetPartMaterialsDynamic _asfort_player_pawn "/BRCosmetics/Characters/Player/Male/Medium/Bodies/M_MED_Faux/Materials/transition/MI_F_Faux_unzip_faceAcc.MI_F_Faux_unzip_faceAcc" (Utilities|Select :Index (LocalVariables|GetCurrent%20Material%20Variant) :Option%200 "/BRCosmetics/Characters/Player/Male/Medium/Bodies/M_MED_Faux/Materials/transition/MI_F_Faux_unzip_head.MI_F_Faux_unzip_head" :Option%201 "/BRCosmetics/Characters/Player/Male/Medium/Bodies/M_MED_Faux/Materials/transition/MI_F_Faux_unzip_head_Cream.MI_F_Faux_unzip_head_Cream") (Utilities|Select :Index (LocalVariables|GetCurrent%20Material%20Variant) :Option%200 "/BRCosmetics/Characters/Player/Male/Medium/Bodies/M_MED_Faux/Materials/transition/MI_F_Faux_unzip_body.MI_F_Faux_unzip_body" :Option%201 "/BRCosmetics/Characters/Player/Male/Medium/Bodies/M_MED_Faux/Materials/transition/MI_F_Faux_unzip_body_Cream.MI_F_Faux_unzip_body_Cream"))))
    (:then_3
      (return true))))

(fn Received_NotifyEnd
  (params (MeshComp :type Skeletal%20Mesh%20Component%20Object%20Reference) (Animation :type Anim%20Sequence%20Base%20Object%20Reference) (EventReference :type Anim%20Notify%20Event%20Reference%20Structure%20%28by%20ref%29))
  (if (Variables|Default|GetReset%20Materials%20When%20Done)
    (bind _asfort_player_pawn (Utilities|Casting|CastToFortPlayerPawn (Components|GetOwner MeshComp)))
    (Utilities|FlowControl|Sequence
      (:then_0
        (Utilities|IsValid (Customization|GetSkeletalMeshforPartType _asfort_player_pawn "Head" :!class /Script/FortniteGame.FortPlayerPawn)
          (:Is%20Valid
            (if (Utilities|Equal(Object) "/BRCosmetics/Characters/Player/Female/Medium/Heads/F_MED_Faux_Head/Meshes/Female_Medium_Faux_Head.Female_Medium_Faux_Head" (Variables|/Script/Engine.SkinnedMeshComponent|GetSkeletalMesh (Customization|GetSkeletalMeshforPartType _asfort_player_pawn "Head" :!class /Script/FortniteGame.FortPlayerPawn)))
              (LocalVariables|SetCurrent%20Progressive%20Variant 1)))))
      (:then_1
        (bind _index (CosmeticLoadout|GetVariantChannelActiveIndexforSlotActor _asfort_player_pawn "(TagName=\"LoadoutSlot.Character\")" _asfort_player_pawn "(TagName=\"Cosmetics.Variant.Channel.Material\")" (Utilities|Array|MakeArray "(TagName=\"Cosmetics.Variant.Property.Mat1\")" "(TagName=\"Cosmetics.Variant.Property.Mat2\")")))
        (LocalVariables|SetCurrent%20Material%20Variant _index))
      (:then_2
        (if (== (LocalVariables|GetCurrent%20Progressive%20Variant) 1)
          (CallFunction|SetPartMaterials _asfort_player_pawn "/BRCosmetics/Characters/Player/Female/Medium/Bodies/F_MED_Faux/Materials/F_MED_Faux_Hair.F_MED_Faux_Hair" (Utilities|Select :Index (LocalVariables|GetCurrent%20Material%20Variant) :Option%200 "/BRCosmetics/Characters/Player/Female/Medium/Heads/F_MED_Faux_Head/Materials/F_MED_Faux_Head.F_MED_Faux_Head" :Option%201 "/BRCosmetics/Characters/Player/Female/Medium/Heads/F_MED_Faux_Head/Skin/Cream/Materials/F_MED_Faux_Cream_Head.F_MED_Faux_Cream_Head") (Utilities|Select :Index (LocalVariables|GetCurrent%20Material%20Variant) :Option%200 "/BRCosmetics/Characters/Player/Female/Medium/Bodies/F_MED_Faux/Materials/F_MED_Faux_Body.F_MED_Faux_Body" :Option%201 "/BRCosmetics/Characters/Player/Female/Medium/Bodies/F_MED_Faux/Skins/Cream/Materials/F_MED_Faux_Cream_Body.F_MED_Faux_Cream_Body"))
          (Development|PrintString 0 "Restoring Faux alien mats")))
      (:then_3
        (return true)))))

(fn SetPartMaterials
  (params (Pawn :type Fort%20Player%20Pawn%20Object%20Reference) (Face%20Material :type Material%20Interface%20Object%20Reference) (Head%20Material :type Material%20Interface%20Object%20Reference) (Body%20Material :type Material%20Interface%20Object%20Reference))
  (Utilities|IsValid Pawn
    (:Is%20Valid
      (Utilities|FlowControl|Sequence
        (:then_0
          (Rendering|Material|SetMaterial (Customization|GetSkeletalMeshforPartType Pawn "Face" :!class /Script/FortniteGame.FortPlayerPawn) 0 Face%20Material))
        (:then_1
          (Rendering|Material|SetMaterial (Customization|GetSkeletalMeshforPartType Pawn "Head" :!class /Script/FortniteGame.FortPlayerPawn) 0 Head%20Material))
        (:then_2
          (Rendering|Material|SetMaterial (Customization|GetSkeletalMeshforPartType Pawn "Body" :!class /Script/FortniteGame.FortPlayerPawn) 0 Body%20Material))))))

(fn SetPartMaterialsDynamic
  (params (Player%20Pawn :type Fort%20Player%20Pawn%20Object%20Reference) (Face%20Material%20Base :type Material%20Interface%20Object%20Reference) (Head%20Material%20Base :type Material%20Interface%20Object%20Reference) (Body%20Material%20Base :type Material%20Interface%20Object%20Reference))
  (bind _returnvalue (Rendering|Material|CreateDynamicMaterialInstance Player%20Pawn Face%20Material%20Base "Alien_Transition_Face" "None" :!class /Script/Engine.KismetMaterialLibrary))
  (bind _returnvalue_2 (Rendering|Material|CreateDynamicMaterialInstance Player%20Pawn Head%20Material%20Base "Alien_Transition_Head" "None" :!class /Script/Engine.KismetMaterialLibrary))
  (bind _returnvalue_3 (Rendering|Material|CreateDynamicMaterialInstance Player%20Pawn Body%20Material%20Base "Alien_Transition_Body" "None" :!class /Script/Engine.KismetMaterialLibrary))
  (CallFunction|SetPartMaterials Player%20Pawn _returnvalue _returnvalue_2 _returnvalue_3))
