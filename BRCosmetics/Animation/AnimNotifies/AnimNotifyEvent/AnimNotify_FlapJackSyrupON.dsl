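; Anim notify state for FlapJack syrup that, when it ends, timestamps the character's materials with the current game time.
; Received_NotifyEnd reads the active Progressive cosmetic variant from the owning FortPlayerPawn's character loadout slot, defaulting to 0 if invalid.
; A switch on that index selects the hat (MiscOrTail) or head (Face) mesh to receive the FJ_GameTimeFromBP scalar parameter.
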(fn Received_NotifyEnd
  (params (MeshComp :type Skeletal%20Mesh%20Component%20Object%20Reference) (Animation :type Anim%20Sequence%20Base%20Object%20Reference) (EventReference :type Anim%20Notify%20Event%20Reference%20Structure%20%28by%20ref%29))
  (Utilities|FlowControl|Sequence
    (:then_0
      ;[1 1 1] Get Variant
      (bind _asfort_player_pawn (Utilities|Casting|CastToFortPlayerPawn (Components|GetOwner MeshComp))
        (:then
          (bind (_index _bindexisvalid) (CosmeticLoadout|GetVariantChannelActiveIndexforSlotActor 0 "(TagName=\"LoadoutSlot.Character\")" _asfort_player_pawn "(TagName=\"Cosmetics.Variant.Channel.Progressive\")" (Utilities|Array|MakeArray "(TagName=\"Cosmetics.Variant.Property.Stage0\")" "(TagName=\"Cosmetics.Variant.Property.Stage4\")")))
          (if _bindexisvalid
            (LocalVariables|SetCurrent%20Variant _index)
            (else
              (LocalVariables|SetCurrent%20Variant 0))))
        (:CastFailed
          (Development|PrintString 0 "FortPlayerPawn CAST FAILED"))))
    (:then_1
      (bind _asfort_player_pawn_2 (Utilities|Casting|CastToFortPlayerPawn (Components|GetOwner MeshComp)))
      ;[1 1 1] Switch between hat and Head Variant
      (switch int (LocalVariables|GetCurrent%20Variant)
        (:Default
          _)
        (:0
          _)
        (:1
          (Rendering|Material|SetScalarParameterValueonMaterials (Customization|GetSkeletalMeshforPartType _asfort_player_pawn_2 "MiscOrTail" :!class /Script/FortniteGame.FortPlayerPawn) "FJ_GameTimeFromBP" (Utilities|Time|GetGameTimeinSeconds _asfort_player_pawn_2))))
      (Rendering|Material|SetScalarParameterValueonMaterials (Customization|GetSkeletalMeshforPartType _asfort_player_pawn_2 "Face" :!class /Script/FortniteGame.FortPlayerPawn) "FJ_GameTimeFromBP" (Utilities|Time|GetGameTimeinSeconds _asfort_player_pawn_2)))
    (:then_2
      (return false))))
