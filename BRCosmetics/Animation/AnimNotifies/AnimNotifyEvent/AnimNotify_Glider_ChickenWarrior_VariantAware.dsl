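; Anim notify for the Chicken Warrior glider that spawns a variant-specific effect on the mesh's Root socket.
; In the frontend, it compares the mesh's first material, or its dynamic instance's parent, against three variant materials to derive a cosmetic index.
; Slurp gets its own effect, while default and dark share one driven by a VariantNum parameter.
; Outside the frontend, the index comes from the pawn's particle variant channel, a path commented as possibly unnecessary.
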(fn Received_Notify
  (params (MeshComp :type Skeletal%20Mesh%20Component%20Object%20Reference) (Animation :type Anim%20Sequence%20Base%20Object%20Reference) (EventReference :type Anim%20Notify%20Event%20Reference%20Structure%20%28by%20ref%29))
  (bind _mesh_comp (LocalVariables|GetMesh_Comp))
  (bind _returnvalue (Rendering|Material|GetMaterials _mesh_comp))
  (bind _output (Utilities|Array|Get(acopy) _returnvalue))
  (bind _output_get (LocalVariables|SetMesh_Comp MeshComp))
  (bind _asb_athena_glider_generic (Utilities|Casting|CastToB_Athena_Glider_Generic (Components|GetOwner _output_get))
    (:then
      (if (Reactive|Frontend|IsCosmeticObjectDisplayedinFrontendContext _asb_athena_glider_generic)
        (for (_array_element _array_index) (Utilities|Array|MakeArray "/BRCosmetics/Gadgets/Assets/VinderTech_GliderChute/Glider_ChickenWarrior_Male/Materials/MI_Glider_ChickenWarrior_Male.MI_Glider_ChickenWarrior_Male" "/BRCosmetics/Gadgets/Assets/VinderTech_GliderChute/Glider_ChickenWarrior_Male/Materials/MI_Glider_ChickenWarrior_Dark_Male.MI_Glider_ChickenWarrior_Dark_Male" "/BRCosmetics/Gadgets/Assets/VinderTech_GliderChute/Glider_ChickenWarrior_Male/Materials/MI_Glider_ChickenWarrior_Slurp_Male.MI_Glider_ChickenWarrior_Slurp_Male")
          (bind _asmaterial_instance_dynamic (Utilities|Casting|CastToMaterialInstanceDynamic _output)
            (:then
              (if (not (Utilities|Equal(Object) (Variables|/Script/Engine.MaterialInstance|GetParent _asmaterial_instance_dynamic) _array_element))
                ;[1 1 1] debug
                (Development|PrintString "Invalid Material Instance Dynamic Index")))
            (:CastFailed
              (if (not (Utilities|Equal(Object) _output _array_element))
                (Development|PrintString "Invalid Material Index"))))
          (bind _output_get_2 (LocalVariables|SetCosmetic%20Index _array_index))
          (switch int _output_get_2
            (:0
              _)
            (:1
              _)
            (:2
              (Effects|SpawnFXSystemAttached :FXSystemAsset (Variables|Default|GetSlurp%20Chicken%20Effect) :AttachToComponent (LocalVariables|GetMesh_Comp) :AttachPointName "Root" :Scale "1.000000,1.000000,1.000000" :LocationType "SnapToTarget" :bAutoDestroy true :PoolingMethod "None" :bAutoActivate true)
              (return false)))
          (bind _returnvalue_2 (Effects|SpawnFXSystemAttached :FXSystemAsset (Variables|Default|GetDefault/Dark%20Chicken%20Varient%20Effect) :AttachToComponent (LocalVariables|GetMesh_Comp) :AttachPointName "Root" :Scale "1.000000,1.000000,1.000000" :LocationType "SnapToTarget" :bAutoDestroy true :PoolingMethod "None" :bAutoActivate true))
          (Effects|Components|ParticleSystem|SetIntParameter _returnvalue_2 "VariantNum" (LocalVariables|GetCosmetic%20Index)))
        (else
          ;[1 1 1] Might be completly unessessary now that we just compare the skeletal meshes material (varaint aware)
          (bind _asfort_player_pawn_athena (Utilities|Casting|CastToFortPlayerPawnAthena (Class|FortPlayerParachute|GetFortPlayerPawn _asb_athena_glider_generic))
            (:then
              (bind _index (CosmeticLoadout|GetVariantChannelActiveIndexforSlotActor "(TagName=\"LoadoutSlot.Glider\")" _asfort_player_pawn_athena "(TagName=\"Cosmetics.Variant.Channel.Particle\")" (Utilities|Array|MakeArray "(TagName=\"Cosmetics.Variant.Property.Particle1\")" "(TagName=\"Cosmetics.Variant.Property.Particle2\")" "(TagName=\"Cosmetics.Variant.Property.Particle3\")")))
              (bind _output_get_3 (LocalVariables|SetCosmetic%20Index _index))
              (switch int _output_get_3
                (:0
                  _)
                (:1
                  _)
                (:2
                  _)))
            (:CastFailed
              ;;[1 1 1] debug
              (Development|PrintString "IS NOT FORTPLAYERPAWNATHENA"))))))
    (:CastFailed
      ;[1 1 1] debug
      (Development|PrintString "IS NOT B_Athena_Gldier_Generic"))))
