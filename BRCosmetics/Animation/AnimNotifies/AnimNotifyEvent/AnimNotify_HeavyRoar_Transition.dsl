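; Anim notify state that swaps HeavyRoar skin materials to transition versions for the notify's duration, based on the active material variant.
; Begin creates transition dynamic instances, applies the Body one, and handles variants 0 and 2 by setting "Hide Element 01".
; Variant 2 also gets transition materials on both Face slots.
; End re-reads the variant and applies fresh base Body or Style Body, face-accessory and head materials for those two variants.
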(fn Received_NotifyBegin
  (params (MeshComp :type Skeletal%20Mesh%20Component%20Object%20Reference) (Animation :type Anim%20Sequence%20Base%20Object%20Reference) (TotalDuration :type Float%20%28single-precision%29) (EventReference :type Anim%20Notify%20Event%20Reference%20Structure%20%28by%20ref%29))
  (bind _bodystylemid (LocalVariables|GetBodyStyleMID))
  (Utilities|FlowControl|Sequence
    (:then_0
      (bind _asfort_player_pawn (Utilities|Casting|CastToFortPlayerPawn (Components|GetOwner MeshComp)))
      (bind _output_get (LocalVariables|SetFortPlayerPawn _asfort_player_pawn))
      (bind _index (CosmeticLoadout|GetVariantChannelActiveIndexforSlotActor 0 "(TagName=\"LoadoutSlot.Character\")" _output_get "(TagName=\"Cosmetics.Variant.Channel.Material\")" (Utilities|Array|MakeArray "(TagName=\"Cosmetics.Variant.Property.Mat1\")" "(TagName=\"Cosmetics.Variant.Property.Mat2\")" "(TagName=\"Cosmetics.Variant.Property.Mat3\")")))
      (bind _output_get_2 (LocalVariables|SetCurrentVariant _index)))
    (:then_1
      (bind _returnvalue (Rendering|Material|CreateDynamicMaterialInstance MeshComp "/BRCosmetics/Characters/Player/Male/Medium/Bodies/M_MED_HeavyRoar_Style/Materials/M_MED_HeavyRoar_Body_Style_Transition.M_MED_HeavyRoar_Body_Style_Transition" "Body" "None" :!class /Script/Engine.KismetMaterialLibrary))
      (LocalVariables|SetBodyStyleMID _returnvalue)
      (bind _returnvalue_2 (Rendering|Material|CreateDynamicMaterialInstance MeshComp "/BRCosmetics/Characters/Player/Male/Medium/Bodies/M_MED_HeavyRoar_Style/Materials/M_MED_HeavyRoar_Style_FaceAcc_Transition.M_MED_HeavyRoar_Style_FaceAcc_Transition" "FaceAcc_Style" "None" :!class /Script/Engine.KismetMaterialLibrary))
      (LocalVariables|SetFaceAccStyleMID _returnvalue_2)
      (bind _returnvalue_3 (Rendering|Material|CreateDynamicMaterialInstance MeshComp "/BRCosmetics/Characters/Player/Male/Medium/Bodies/M_MED_HeavyRoar_Style/Materials/M_MED_HeavyRoar_Style_Head_Transition.M_MED_HeavyRoar_Style_Head_Transition" "Head_Style" "None" :!class /Script/Engine.KismetMaterialLibrary))
      (LocalVariables|SetHeadStyleMID _returnvalue_3))
    (:then_2
      (Rendering|Material|SetMaterial (Customization|GetSkeletalMeshforPartType (LocalVariables|GetFortPlayerPawn) "Body" :!class /Script/FortniteGame.FortPlayerPawn) 0 _bodystylemid)
      (switch int (LocalVariables|GetCurrentVariant)
        (:0
          (Rendering|Material|SetScalarParameterValue _bodystylemid "Hide Element 01" 0.0 :!class /Script/Engine.MaterialInstanceDynamic))
        (:2
          (Rendering|Material|SetScalarParameterValue _bodystylemid "Hide Element 01" 1.0 :!class /Script/Engine.MaterialInstanceDynamic)
          (Rendering|Material|SetMaterial (Customization|GetSkeletalMeshforPartType (LocalVariables|GetFortPlayerPawn) "Face" :!class /Script/FortniteGame.FortPlayerPawn) 0 (LocalVariables|GetFaceAccStyleMID))
          (Rendering|Material|SetMaterial (Customization|GetSkeletalMeshforPartType (LocalVariables|GetFortPlayerPawn) "Face" :!class /Script/FortniteGame.FortPlayerPawn) 1 (LocalVariables|GetHeadStyleMID)))))))

(fn Received_NotifyEnd
  (params (MeshComp :type Skeletal%20Mesh%20Component%20Object%20Reference) (Animation :type Anim%20Sequence%20Base%20Object%20Reference) (EventReference :type Anim%20Notify%20Event%20Reference%20Structure%20%28by%20ref%29))
  (Utilities|FlowControl|Sequence
    (:then_0
      (bind _asfort_player_pawn (Utilities|Casting|CastToFortPlayerPawn (Components|GetOwner MeshComp)))
      (bind _output_get (LocalVariables|SetFortPlayerPawn _asfort_player_pawn))
      (bind _index (CosmeticLoadout|GetVariantChannelActiveIndexforSlotActor 0 "(TagName=\"LoadoutSlot.Character\")" _output_get "(TagName=\"Cosmetics.Variant.Channel.Material\")" (Utilities|Array|MakeArray "(TagName=\"Cosmetics.Variant.Property.Mat1\")" "(TagName=\"Cosmetics.Variant.Property.Mat2\")" "(TagName=\"Cosmetics.Variant.Property.Mat3\")")))
      (LocalVariables|SetCurrentVariant _index))
    (:then_1
      (switch int (LocalVariables|GetCurrentVariant)
        (:0
          (bind _returnvalue (Rendering|Material|CreateDynamicMaterialInstance MeshComp "/BRCosmetics/Characters/Player/Male/Medium/Bodies/M_MED_HeavyRoar/Materials/M_MED_HeavyRoar_Body.M_MED_HeavyRoar_Body" "Body" "None" :!class /Script/Engine.KismetMaterialLibrary))
          (Rendering|Material|SetMaterial (Customization|GetSkeletalMeshforPartType (LocalVariables|GetFortPlayerPawn) "Body" :!class /Script/FortniteGame.FortPlayerPawn) 0 _returnvalue)
          (Rendering|Material|SetScalarParameterValue _returnvalue "Hide Element 01" 0.0 :!class /Script/Engine.MaterialInstanceDynamic))
        (:2
          (Utilities|FlowControl|Sequence
            (:then_0
              (bind _returnvalue_2 (Rendering|Material|CreateDynamicMaterialInstance MeshComp "/BRCosmetics/Characters/Player/Male/Medium/Bodies/M_MED_HeavyRoar_Style/Materials/M_MED_HeavyRoar_Body_Style.M_MED_HeavyRoar_Body_Style" "Body Style" "None" :!class /Script/Engine.KismetMaterialLibrary))
              (Rendering|Material|SetMaterial (Customization|GetSkeletalMeshforPartType (LocalVariables|GetFortPlayerPawn) "Body" :!class /Script/FortniteGame.FortPlayerPawn) 0 _returnvalue_2)
              (Rendering|Material|SetScalarParameterValue _returnvalue_2 "Hide Element 01" 1.0 :!class /Script/Engine.MaterialInstanceDynamic))
            (:then_1
              (bind _returnvalue_3 (Rendering|Material|CreateDynamicMaterialInstance MeshComp "/BRCosmetics/Characters/Player/Male/Medium/Bodies/M_MED_HeavyRoar_Style/Materials/M_MED_HeavyRoar_Style_FaceAcc.M_MED_HeavyRoar_Style_FaceAcc" "FaceAcc Style" "None" :!class /Script/Engine.KismetMaterialLibrary))
              (Rendering|Material|SetMaterial (Customization|GetSkeletalMeshforPartType (LocalVariables|GetFortPlayerPawn) "Face" :!class /Script/FortniteGame.FortPlayerPawn) 0 _returnvalue_3))
            (:then_2
              (bind _returnvalue_4 (Rendering|Material|CreateDynamicMaterialInstance MeshComp "/BRCosmetics/Characters/Player/Male/Medium/Bodies/M_MED_HeavyRoar_Style/Materials/M_MED_HeavyRoar_Style_Head.M_MED_HeavyRoar_Style_Head" "Head Style" "None" :!class /Script/Engine.KismetMaterialLibrary))
              (Rendering|Material|SetMaterial (Customization|GetSkeletalMeshforPartType (LocalVariables|GetFortPlayerPawn) "Face" :!class /Script/FortniteGame.FortPlayerPawn) 1 _returnvalue_4))))))))
