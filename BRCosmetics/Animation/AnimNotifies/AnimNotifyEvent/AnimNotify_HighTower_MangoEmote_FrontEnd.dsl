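; Front-end anim notify for the HighTower Mango emote that swaps the emoting pawn's cosmetics to copy another team member on the lobby pedestals.
; It picks the other pawn deterministically, ranking character item names by a character-sum hash mod 257, with display-name hash as tiebreaker.
; Character, backpack and extra-part items are cloned from that pawn.
; Non-pedestal pawns, or those with no candidate, get the DefaultBase outfit.
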(fn GetDefaultOutfit
  (params (Selection :type EDefaultCharacterSelection%20Enum :default "Random"))
  (return (Utilities|Array|MakeArray (Class|FortCosmeticFunctionLibrary|MakeSwapDataforCosmetic (Class|FortGameDataCosmeticsHelper|GetGameDefaultOutfit Selection)))))

(fn GetStringHashValue
  (params (Input :type String%20%28by%20ref%29))
  (for _index (range (Utilities|String|Len Input))
    (LocalVariables|SetAccumulator (+ (LocalVariables|GetAccumulator) (Utilities|String|GetCharacterasNumber Input _index))))
  (return (Math|Integer|%25%28Integer%29 (LocalVariables|GetAccumulator) 257)))

(fn GetSwapDataDeterministic
  (params (Pawn :type Fort%20Player%20Pawn%20Object%20Reference%20%28by%20ref%29) (Pedestals :type Array%20of%20Fort%20Team%20Member%20Pedestal%20Object%20References))
  (bind _bestcidhash (LocalVariables|GetBestCIDHash))
  (for _array_element Pedestals
    ;[1 1 1] Deterministically figure out the best pawn to mimic the skin from
    (Utilities|IsValid (Variables|/Script/FortniteUI.FortItemPlacementActor|GetCurrentPawn _array_element)
      (:Is%20Valid
        (if (Utilities|NotEqual(Object) (Variables|/Script/FortniteUI.FortItemPlacementActor|GetCurrentPawn _array_element) Pawn)
          (if (> (CallFunction|GetStringHashValue (Item|GetPersistentName (CosmeticLoadout|GetItemDefinitionfromLoadoutwithActor (Variables|/Script/FortniteUI.FortItemPlacementActor|GetCurrentPawn _array_element) "(TagName=\"LoadoutSlot.Character\")"))) _bestcidhash)
            _
            (elif (== (CallFunction|GetStringHashValue (Item|GetPersistentName (CosmeticLoadout|GetItemDefinitionfromLoadoutwithActor (Variables|/Script/FortniteUI.FortItemPlacementActor|GetCurrentPawn _array_element) "(TagName=\"LoadoutSlot.Character\")"))) _bestcidhash)
              (if (> (CallFunction|GetStringHashValue (Class|FortTeamMemberPedestal|GetTeamMemberDisplayName _array_element)) (LocalVariables|GetBestNameHash))
                _)))
          (LocalVariables|SetBestCIDHash (CallFunction|GetStringHashValue (Item|GetPersistentName (CosmeticLoadout|GetItemDefinitionfromLoadoutwithActor (Variables|/Script/FortniteUI.FortItemPlacementActor|GetCurrentPawn _array_element) "(TagName=\"LoadoutSlot.Character\")"))))
          (LocalVariables|SetBestNameHash (CallFunction|GetStringHashValue (Class|FortTeamMemberPedestal|GetTeamMemberDisplayName _array_element)))
          (LocalVariables|SetBestPawn (Variables|/Script/FortniteUI.FortItemPlacementActor|GetCurrentPawn _array_element))))))
  ;[1 1 1] Generate and return the swap data
  (Utilities|IsValid (LocalVariables|GetBestPawn)
    (:Is%20Valid
      (for _array_element_2 (GameplayTags|BreakGameplayTagContainer (GameplayTags|MakeLiteralGameplayTagContainer "(GameplayTags=((TagName=\"LoadoutSlot.Character\"),(TagName=\"LoadoutSlot.Backpack\"),(TagName=\"LoadoutSlot.ExtraPart\")))"))
        (bind _swapdata (CosmeticLoadout|ConstructItemSwapClonedfromActor (LocalVariables|GetBestPawn) (CosmeticLoadout|GetItemDefinitionfromLoadoutwithActor (LocalVariables|GetBestPawn) _array_element_2)))
        (Utilities|Array|Add (LocalVariables|GetSwapData) _swapdata))
      (return (LocalVariables|GetSwapData)))
    (:Is%20Not%20Valid
      (bind _array (CallFunction|GetDefaultOutfit "DefaultBase"))
      (return _array))))

(fn IsPedestalPawn
  (params (Pawn :type Fort%20Player%20Pawn%20Object%20Reference%20%28by%20ref%29) (Pedestals :type Array%20of%20Fort%20Team%20Member%20Pedestal%20Object%20References))
  (for _array_element Pedestals
    (if (Utilities|Equal(Object) (Variables|/Script/FortniteUI.FortItemPlacementActor|GetCurrentPawn _array_element) Pawn)
      (return true)))
  (return false))

(fn Received_Notify
  (params (MeshComp :type Skeletal%20Mesh%20Component%20Object%20Reference) (Animation :type Anim%20Sequence%20Base%20Object%20Reference) (EventReference :type Anim%20Notify%20Event%20Reference%20Structure%20%28by%20ref%29))
  (bind _asfort_player_pawn (Utilities|Casting|CastToFortPlayerPawn (Components|GetOwner MeshComp)))
  (bind _outactors (Actor|GetAllActorsOfClass "/Script/FortniteUI.FortTeamMemberPedestal"))
  (bind _output_get (LocalVariables|SetPedestals _outactors))
  (bind _bresult (CallFunction|IsPedestalPawn _asfort_player_pawn _output_get))
  (if _bresult
    (bind _result (CallFunction|GetSwapDataDeterministic _asfort_player_pawn (LocalVariables|GetPedestals)))
    (Class|FortPlayerPawn|SwapPawnCosmeticItemandVariant _asfort_player_pawn (Variables|Default|GetSwapId) _result false "None" false)
    (else
      (bind _array (CallFunction|GetDefaultOutfit "DefaultBase"))
      (Class|FortPlayerPawn|SwapPawnCosmeticItemandVariant _asfort_player_pawn (Variables|Default|GetSwapId) _array false "None" false)))
  (return false))
