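; Anim notify that spawns the NPC variant of the HighTower Tomato emote thruster Niagara effect during an animation.
; It casts the mesh component's owner to FortPlayerPawn and fetches that pawn's Body part skeletal mesh, rather than using the notifying mesh directly.
; The system attaches to that mesh's root with auto-release pooling, and the notify returns false.
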(fn Received_Notify
  (params (MeshComp :type Skeletal%20Mesh%20Component%20Object%20Reference) (Animation :type Anim%20Sequence%20Base%20Object%20Reference) (EventReference :type Anim%20Notify%20Event%20Reference%20Structure%20%28by%20ref%29))
  (bind _asfort_player_pawn (Utilities|Casting|CastToFortPlayerPawn (Components|GetOwner MeshComp)))
  (Niagara|SpawnSystemAttached "/BRCosmetics/Characters/Player/Male/Medium/Bodies/M_MED_HighTower_Tomato_Casual/FX/NS_HighTower_Tomato_Emote_ThrusterFX_NPC.NS_HighTower_Tomato_Emote_ThrusterFX_NPC" (Customization|GetSkeletalMeshforPartType _asfort_player_pawn "Body" :!class /Script/FortniteGame.FortPlayerPawn) "root" "0, 0, 0" "0, 0, 0" "KeepRelativeOffset" true true "AutoRelease")
  (return false))
