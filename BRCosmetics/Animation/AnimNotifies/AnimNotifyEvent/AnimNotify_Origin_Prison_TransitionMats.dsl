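; Anim notify state that swaps body materials on the Origin Prison outfit during a transition animation.
; Both ends wait for the Head part, proceed only for the Origin Prison head mesh, and pick the suit material from the active Material variant.
; Begin puts dynamic string and suit instances on body slots 0 and 1.
; End, if Reset Materials When Done is set, applies vertex-crunch and static suit materials and prints a debug string.
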(fn Received_NotifyBegin
  (params (MeshComp :type Skeletal%20Mesh%20Component%20Object%20Reference) (Animation :type Anim%20Sequence%20Base%20Object%20Reference) (TotalDuration :type Float%20%28single-precision%29) (EventReference :type Anim%20Notify%20Event%20Reference%20Structure%20%28by%20ref%29))
  (bind _asfort_player_pawn (Utilities|Casting|CastToFortPlayerPawn (Components|GetOwner MeshComp)))
  (Cosmetics|WhenCosmeticSpecificPartApplied _asfort_player_pawn (EventDispatchers|CreateEvent :!event Received_NotifyBegin_WhenSpecificPart) false "Head")
  (return true))

(fn Received_NotifyBegin_WhenSpecificPart
  (params (Pawn :type Fort%20Player%20Pawn%20Object%20Reference) (Part :type Custom%20Character%20Part%20Object%20Reference) (MeshPart :type Skeletal%20Mesh%20Component%20Object%20Reference) (PartType :type EFortCustomPartType%20Enum :default "Head"))
  (Utilities|FlowControl|Sequence
    (:then_0
      (Utilities|IsValid MeshPart
        (:Is%20Valid
          (if (Utilities|Equal(Object) "/BRCosmetics/Characters/Player/Male/Medium/Heads/M_MED_Origin_Prison_Head/Meshes/Male_Medium_Origin_Prison_Head.Male_Medium_Origin_Prison_Head" (Variables|/Script/Engine.SkinnedMeshComponent|GetSkeletalMesh MeshPart))
            (LocalVariables|SetCurrent%20Progressive%20Variant 1)))))
    (:then_1
      (bind _index (CosmeticLoadout|GetVariantChannelActiveIndexforSlotActor Pawn "(TagName=\"LoadoutSlot.Character\")" Pawn "(TagName=\"Cosmetics.Variant.Channel.Material\")" (Utilities|Array|MakeArray "(TagName=\"Cosmetics.Variant.Property.Mat1\")" "(TagName=\"Cosmetics.Variant.Property.Mat2\")" "(TagName=\"Cosmetics.Variant.Property.Mat3\")")))
      (LocalVariables|SetCurrent%20Material%20Variant _index))
    (:then_2
      (LocalVariables|SetString%20Transition%20Material%20Base "/BRCosmetics/Characters/Player/Male/Medium/Bodies/M_MED_Origin_Prison/Materials/TechArt/MAT_OriginPrison_String_Inst.MAT_OriginPrison_String_Inst")
      (LocalVariables|SetSuit%20Transition%20Material%20Base "/BRCosmetics/Characters/Player/Male/Medium/Bodies/M_MED_Origin_Prison/Materials/M_OriginPrison_Body.M_OriginPrison_Body"))
    (:then_3
      (if (== (LocalVariables|GetCurrent%20Progressive%20Variant) 1)
        (CallFunction|SetPartMaterialsDynamic Pawn (LocalVariables|GetString%20Transition%20Material%20Base) (Utilities|Select :Index (LocalVariables|GetCurrent%20Material%20Variant) :Option%200 "/BRCosmetics/Characters/Player/Male/Medium/Bodies/M_MED_Origin_Prison/Materials/M_OriginPrison_Body.M_OriginPrison_Body" :Option%201 "/BRCosmetics/Characters/Player/Male/Medium/Bodies/M_MED_Origin_Prison/Materials/M_OriginPrisonGrey_Body.M_OriginPrisonGrey_Body" :Option%202 "/BRCosmetics/Characters/Player/Male/Medium/Bodies/M_MED_Origin_Prison/Materials/M_OriginPrisonDark_Body.M_OriginPrisonDark_Body"))))))

(fn Received_NotifyEnd
  (params (MeshComp :type Skeletal%20Mesh%20Component%20Object%20Reference) (Animation :type Anim%20Sequence%20Base%20Object%20Reference) (EventReference :type Anim%20Notify%20Event%20Reference%20Structure%20%28by%20ref%29))
  (if (Variables|Default|GetReset%20Materials%20When%20Done)
    (bind _asfort_player_pawn (Utilities|Casting|CastToFortPlayerPawn (Components|GetOwner MeshComp)))
    (Cosmetics|WhenCosmeticSpecificPartApplied _asfort_player_pawn (EventDispatchers|CreateEvent :!event Received_NotifyEnd_WhenSpecificPart) false "Head")
    (return true)))

(fn Received_NotifyEnd_WhenSpecificPart
  (params (Pawn :type Fort%20Player%20Pawn%20Object%20Reference) (Part :type Custom%20Character%20Part%20Object%20Reference) (MeshPart :type Skeletal%20Mesh%20Component%20Object%20Reference) (PartType :type EFortCustomPartType%20Enum :default "Head"))
  (Utilities|FlowControl|Sequence
    (:then_0
      (Utilities|IsValid MeshPart
        (:Is%20Valid
          (if (Utilities|Equal(Object) "/BRCosmetics/Characters/Player/Male/Medium/Heads/M_MED_Origin_Prison_Head/Meshes/Male_Medium_Origin_Prison_Head.Male_Medium_Origin_Prison_Head" (Variables|/Script/Engine.SkinnedMeshComponent|GetSkeletalMesh MeshPart))
            (LocalVariables|SetCurrent%20Progressive%20Variant 1)))))
    (:then_1
      (bind _index (CosmeticLoadout|GetVariantChannelActiveIndexforSlotActor Pawn "(TagName=\"LoadoutSlot.Character\")" Pawn "(TagName=\"Cosmetics.Variant.Channel.Material\")" (Utilities|Array|MakeArray "(TagName=\"Cosmetics.Variant.Property.Mat1\")" "(TagName=\"Cosmetics.Variant.Property.Mat2\")" "(TagName=\"Cosmetics.Variant.Property.Mat3\")")))
      (LocalVariables|SetCurrent%20Material%20Variant _index))
    (:then_2
      (LocalVariables|SetVertex%20Crunch "/Game/Characters/Player/Male/Medium/Bodies/M_MED_HighTower_Tomato_Casual/Materials/M_VertexCrunch.M_VertexCrunch"))
    (:then_3
      (if (== (LocalVariables|GetCurrent%20Progressive%20Variant) 1)
        (CallFunction|SetPartMaterials Pawn (LocalVariables|GetVertex%20Crunch) (Utilities|Select :Index (LocalVariables|GetCurrent%20Material%20Variant) :Option%200 "/BRCosmetics/Characters/Player/Male/Medium/Bodies/M_MED_Origin_Prison/Materials/M_OriginPrison_Body.M_OriginPrison_Body" :Option%201 "/BRCosmetics/Characters/Player/Male/Medium/Bodies/M_MED_Origin_Prison/Materials/M_OriginPrisonGrey_Body.M_OriginPrisonGrey_Body" :Option%202 "/BRCosmetics/Characters/Player/Male/Medium/Bodies/M_MED_Origin_Prison/Materials/M_OriginPrisonDark_Body.M_OriginPrisonDark_Body"))
        (Development|PrintString 0 "Restoring OG transition mats")))))

(fn SetPartMaterials
  (params (Pawn :type Fort%20Player%20Pawn%20Object%20Reference) (String%20Material :type Material%20Interface%20Object%20Reference) (Suit%20Material :type Material%20Interface%20Object%20Reference))
  (Utilities|IsValid Pawn
    (:Is%20Valid
      (Utilities|FlowControl|Sequence
        (:then_0
          (Rendering|Material|SetMaterial (Customization|GetSkeletalMeshforPartType Pawn "Body" :!class /Script/FortniteGame.FortPlayerPawn) 0 String%20Material))
        (:then_1
          (Rendering|Material|SetMaterial (Customization|GetSkeletalMeshforPartType Pawn "Body" :!class /Script/FortniteGame.FortPlayerPawn) 1 Suit%20Material))))))

(fn SetPartMaterialsDynamic
  (params (Player%20Pawn :type Fort%20Player%20Pawn%20Object%20Reference) (String%20Material%20Base :type Material%20Interface%20Object%20Reference) (Suit%20Material%20Base :type Material%20Interface%20Object%20Reference))
  (bind _returnvalue (Rendering|Material|CreateDynamicMaterialInstance Player%20Pawn String%20Material%20Base "OG_Transition_String" "None" :!class /Script/Engine.KismetMaterialLibrary))
  (bind _returnvalue_2 (Rendering|Material|CreateDynamicMaterialInstance Player%20Pawn Suit%20Material%20Base "OG_Transition_Suit" "None" :!class /Script/Engine.KismetMaterialLibrary))
  (CallFunction|SetPartMaterials Player%20Pawn _returnvalue _returnvalue_2))
