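; Anim notify for a finger-snap emote that spawns snap particles and sound, with special gauntlet variants for certain cosmetics.
; When the mesh matches Jim, it enables six colored flare emitters from the pawn's SpecialActor.Ashton.Object gameplay tags, printing them for debugging.
; Separately, loadout checks add Niagara flares for a specific backpack or the Menace character, avoiding doubled sound when both match.
; The normal snap emitter also spawns in that loadout path.
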(fn Received_Notify
  (params (MeshComp :type Skeletal%20Mesh%20Component%20Object%20Reference) (Animation :type Anim%20Sequence%20Base%20Object%20Reference) (EventReference :type Anim%20Notify%20Event%20Reference%20Structure%20%28by%20ref%29))
  (Utilities|FlowControl|Sequence
    (:then_0
      ;[1 1 1] Gameplay Version
      (if (Utilities|Equal(Object) (Variables|/Script/Engine.SkinnedMeshComponent|GetSkeletalMesh MeshComp) (Variables|Default|GetJim))
        (bind _asplayer_pawn_athena (Utilities|Casting|CastToPlayerPawn_Athena (Components|GetOwner MeshComp)))
        (bind _returnvalue (Effects|SpawnEmitterAttached :EmitterTemplate "/BRCosmetics/Athena/Items/Gameplay/BackPacks/CarminePack/FX/P_Jim_Spawning_Gauntlet_Flare_Snap.P_Jim_Spawning_Gauntlet_Flare_Snap" :AttachToComponent MeshComp :AttachPointName "FX_Snapture" :Scale "1.000000,1.000000,1.000000" :LocationType "SnapToTarget" :bAutoDestroy true :PoolingMethod "None" :bAutoActivate true))
        (Development|PrintString (Utilities|String|Append "B - " (Utilities|String|ToString(Boolean) (GameplayTags|HasMatchingGameplayTag _asplayer_pawn_athena "(TagName=\"SpecialActor.Ashton.Object.B\")")) "; G - " (Utilities|String|ToString(Boolean) (GameplayTags|HasMatchingGameplayTag _asplayer_pawn_athena "(TagName=\"SpecialActor.Ashton.Object.G\")")) "; O - " (Utilities|String|ToString(Boolean) (GameplayTags|HasMatchingGameplayTag _asplayer_pawn_athena "(TagName=\"SpecialActor.Ashton.Object.O\")")) "; P - " (Utilities|String|ToString(Boolean) (GameplayTags|HasMatchingGameplayTag _asplayer_pawn_athena "(TagName=\"SpecialActor.Ashton.Object.P\")")) "; R - " (Utilities|String|ToString(Boolean) (GameplayTags|HasMatchingGameplayTag _asplayer_pawn_athena "(TagName=\"SpecialActor.Ashton.Object.R\")")) "; Y - " (Utilities|String|ToString(Boolean) (GameplayTags|HasMatchingGameplayTag _asplayer_pawn_athena "(TagName=\"SpecialActor.Ashton.Object.Y\")"))) true true "(R=0.000000,G=0.660000,B=1.000000,A=1.000000)" 1e+01)
        (Effects|Components|ParticleSystem|SetEmitterEnable _returnvalue "Blue" (GameplayTags|HasMatchingGameplayTag _asplayer_pawn_athena "(TagName=\"SpecialActor.Ashton.Object.B\")"))
        (Effects|Components|ParticleSystem|SetEmitterEnable _returnvalue "Green" (GameplayTags|HasMatchingGameplayTag _asplayer_pawn_athena "(TagName=\"SpecialActor.Ashton.Object.G\")"))
        (Effects|Components|ParticleSystem|SetEmitterEnable _returnvalue "Orange" (GameplayTags|HasMatchingGameplayTag _asplayer_pawn_athena "(TagName=\"SpecialActor.Ashton.Object.O\")"))
        (Effects|Components|ParticleSystem|SetEmitterEnable _returnvalue "Purple" (GameplayTags|HasMatchingGameplayTag _asplayer_pawn_athena "(TagName=\"SpecialActor.Ashton.Object.P\")"))
        (Effects|Components|ParticleSystem|SetEmitterEnable _returnvalue "Red" (GameplayTags|HasMatchingGameplayTag _asplayer_pawn_athena "(TagName=\"SpecialActor.Ashton.Object.R\")"))
        (Effects|Components|ParticleSystem|SetEmitterEnable _returnvalue "Yellow" (GameplayTags|HasMatchingGameplayTag _asplayer_pawn_athena "(TagName=\"SpecialActor.Ashton.Object.Y\")"))
        (Audio|SpawnSoundAttached "/BRCosmetics/Sounds/Emotes/Athena_Emote_Snap_Carmine_Cue.Athena_Emote_Snap_Carmine_Cue" MeshComp)))
    (:then_1
      (bind _asfort_player_pawn (Utilities|Casting|CastToFortPlayerPawn (Components|GetOwner MeshComp)))
      (LocalVariables|SetisBackpack (Utilities|Equal(Object) (CosmeticLoadout|GetItemDefinitionfromLoadoutwithActor _asfort_player_pawn "(TagName=\"LoadoutSlot.Backpack\")") (Variables|Default|GetBackpack)))
      (LocalVariables|Setis%20Menace (Utilities|Equal(Object) (CosmeticLoadout|GetItemDefinitionfromLoadoutwithActor _asfort_player_pawn "(TagName=\"LoadoutSlot.Character\")") (Variables|Default|GetMenace)))
      (Utilities|FlowControl|Sequence
        (:then_0
          (if (LocalVariables|GetisBackpack)
            (Niagara|SpawnSystemAttached "/BRCosmetics/Athena/Items/Gameplay/BackPacks/CarminePack/FX/NS_Jim_Gauntlet_Flare_Snap_Backpack.NS_Jim_Gauntlet_Flare_Snap_Backpack" (Customization|GetSkeletalMeshforPartType _asfort_player_pawn "Backpack" :!class /Script/FortniteGame.FortPlayerPawn) "pack" "0, 0, 0" "0, 0, 0" "KeepRelativeOffset" true)
            (if (not (LocalVariables|Getis%20Menace))
              (Audio|SpawnSoundAttached "/BRCosmetics/Sounds/Emotes/Athena_Emote_Snap_Carmine_Cue.Athena_Emote_Snap_Carmine_Cue" (Customization|GetSkeletalMeshforPartType _asfort_player_pawn "Backpack" :!class /Script/FortniteGame.FortPlayerPawn)))
            (else
              _))
          ;[1 1 1] Normal Snap
          (Effects|SpawnEmitterAttached :EmitterTemplate "/BRCosmetics/Effects/Fort_Effects/Emotes/P_Snap.P_Snap" :AttachToComponent MeshComp :AttachPointName "FX_Snap" :Scale "1.000000,1.000000,1.000000" :LocationType "SnapToTarget" :bAutoDestroy true :PoolingMethod "None" :bAutoActivate true))
        (:then_1
          (if (LocalVariables|Getis%20Menace)
            (Niagara|SpawnSystemAttached "/BRCosmetics/Athena/Items/Gameplay/BackPacks/CarminePack/FX/NS_Jim_Gauntlet_Flare_Snap.NS_Jim_Gauntlet_Flare_Snap" (Customization|GetSkeletalMeshforPartType _asfort_player_pawn "Body" :!class /Script/FortniteGame.FortPlayerPawn) "root" "0, 0, 0" "0, 0, 0" "KeepRelativeOffset" true)
            (Audio|SpawnSoundAttached "/BRCosmetics/Sounds/Emotes/Athena_Emote_Snap_Carmine_Cue.Athena_Emote_Snap_Carmine_Cue" MeshComp)
            (else
              _)))))
    (:then_2
      (return false))))
