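; Anim notify that clears the Assassin Suit state marker from an Athena player pawn wearing the damaged suit material.
; It checks whether slot 0 of the pawn's Body mesh holds a dynamic material instance parented to M_Med_Assassin_Suit_Damaged.
; On a match it filters the pawn's attached actors for BP_AssassinSuitMark and destroys the first one found.
; Debug "true"/"false" strings print the result of the material comparison.
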(fn Received_Notify
  (params (MeshComp :type Skeletal%20Mesh%20Component%20Object%20Reference) (Animation :type Anim%20Sequence%20Base%20Object%20Reference) (EventReference :type Anim%20Notify%20Event%20Reference%20Structure%20%28by%20ref%29))
  (bind _asplayer_pawn_athena (Utilities|Casting|CastToPlayerPawn_Athena (Components|GetOwner MeshComp)))
  (Utilities|IsValid (Customization|GetSkeletalMeshforPartType _asplayer_pawn_athena "Body" :!class /Script/FortniteGame.FortPlayerPawn)
    (:Is%20Valid
      (if (.bSuccess (Utilities|Casting|CastToMaterialInstanceDynamic (Rendering|Material|GetMaterial (Customization|GetSkeletalMeshforPartType _asplayer_pawn_athena "Body" :!class /Script/FortniteGame.FortPlayerPawn) 0)))
        (if (Utilities|Equal(Object) (Variables|/Script/Engine.MaterialInstance|GetParent (Utilities|Casting|CastToMaterialInstanceDynamic (Rendering|Material|GetMaterial (Customization|GetSkeletalMeshforPartType _asplayer_pawn_athena "Body" :!class /Script/FortniteGame.FortPlayerPawn) 0))) "/BRCosmetics/Characters/Player/Male/Medium/Bodies/M_MED_AssassinSuit/Materials/M_Med_Assassin_Suit_Damaged.M_Med_Assassin_Suit_Damaged")
          (Development|PrintString "true" true true "(R=0.066037,G=1.000000,B=0.000000,A=1.000000)")
          (bind _filteredarray (Utilities|Array|FilterArray (Actor|GetAttachedActors _asplayer_pawn_athena true false) "/BRCosmetics/Effects/Fort_Effects/BluePrints/StateMarker/BP_AssassinSuitMark.BP_AssassinSuitMark_C"))
          (if (Utilities|IsArrayNonEmpty _filteredarray)
            (Actor|DestroyActor (Utilities|Array|Get(acopy) _filteredarray)))
          (else
            (Development|PrintString "false" true true "(R=1.000000,G=0.000000,B=0.297505,A=1.000000)")))
        (return false)
        (else
          _)))
    (:Is%20Not%20Valid
      _)))
