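; Anim notify that stamps the current game time onto a material of the skeletal mesh playing the animation.
; When fired, it gets a dynamic material instance at the configured material index through BTechArtGlobalLibrary.
; If that instance is valid, it writes game time in seconds to the named scalar parameter and returns true.
; Otherwise it returns false without changing anything.
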(fn Received_Notify
  (params (MeshComp :type Skeletal%20Mesh%20Component%20Object%20Reference) (Animation :type Anim%20Sequence%20Base%20Object%20Reference) (EventReference :type Anim%20Notify%20Event%20Reference%20Structure%20%28by%20ref%29))
  (bind _output_get (LocalVariables|SetMID (Class|BTechArtGlobalLibrary|GetDynamicMaterialInstance MeshComp (Variables|Timestamping|GetMaterialIndex))))
  (Utilities|IsValid _output_get
    (:Is%20Valid
      (Rendering|Material|SetScalarParameterValue (LocalVariables|GetMID) (Variables|Timestamping|GetParameterName) (Utilities|Time|GetGameTimeinSeconds) :!class /Script/Engine.MaterialInstanceDynamic)
      (return true))
    (:Is%20Not%20Valid
      (return false))))
