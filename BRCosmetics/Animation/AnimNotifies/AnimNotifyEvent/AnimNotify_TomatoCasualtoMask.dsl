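; Anim notify for the HighTower Tomato casual-to-mask emote, swapping the player's materials for emote versions.
; It reads the character's active Material and Progressive variant indices from the cosmetic loadout.
; If the progressive index matches the expected value, it creates dynamic instances of casual, armor, hair and glasses materials, mostly selected by material variant, and applies them to both slots of Head, Face and Body meshes.
; Otherwise it logs the mismatch and changes nothing.
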(fn Received_Notify
  (params (MeshComp :type Skeletal%20Mesh%20Component%20Object%20Reference) (Animation :type Anim%20Sequence%20Base%20Object%20Reference) (EventReference :type Anim%20Notify%20Event%20Reference%20Structure%20%28by%20ref%29))
  (Utilities|FlowControl|Sequence
    (:then_0
      (bind _asfort_player_pawn (Utilities|Casting|CastToFortPlayerPawn (Components|GetOwner MeshComp))
        (:then
          (bind _index (CosmeticLoadout|GetVariantChannelActiveIndexforSlotActor "(TagName=\"LoadoutSlot.Character\")" _asfort_player_pawn "(TagName=\"Cosmetics.Variant.Channel.Material\")" (Utilities|Array|MakeArray "(TagName=\"Cosmetics.Variant.Property.Mat1\")" "(TagName=\"Cosmetics.Variant.Property.Mat2\")" "(TagName=\"Cosmetics.Variant.Property.Mat3\")" "(TagName=\"Cosmetics.Variant.Property.Mat4\")")))
          (LocalVariables|SetCurrent%20Variant _index)
          (bind _index_2 (CosmeticLoadout|GetVariantChannelActiveIndexforSlotActor "(TagName=\"LoadoutSlot.Character\")" _asfort_player_pawn "(TagName=\"Cosmetics.Variant.Channel.Progressive\")" (Utilities|Array|MakeArray "(TagName=\"Cosmetics.Variant.Property.Stage1\")" "(TagName=\"Cosmetics.Variant.Property.Stage2\")")))
          (LocalVariables|SetCurrent%20Progressive _index_2))
        (:CastFailed
          (Development|PrintString "FortPlayerPawn CAST FAILED"))))
    (:then_1
      (if (== (LocalVariables|GetCurrent%20Progressive) (LocalVariables|GetExpected%20Progressive))
        (Utilities|FlowControl|Sequence
          (:then_0
            (bind _returnvalue (Rendering|Material|CreateDynamicMaterialInstance (Utilities|Select :Index (LocalVariables|GetCurrent%20Variant) :Option%200 "/BRCosmetics/Characters/Player/Male/Medium/Bodies/M_MED_HighTower_Tomato_Casual/Materials/M_MED_HighTower_Tomato_Body_Armor_EMOTE.M_MED_HighTower_Tomato_Body_Armor_EMOTE" :Option%201 "/BRCosmetics/Characters/Player/Male/Medium/Bodies/M_MED_HighTower_Tomato/Materials/M_MED_HighTower_Tomato_Body_Armor_SL_Silver_EMT.M_MED_HighTower_Tomato_Body_Armor_SL_Silver_EMT" :Option%202 "/BRCosmetics/Characters/Player/Male/Medium/Bodies/M_MED_HighTower_Tomato/Materials/M_MED_HighTower_Tomato_Body_Armor_SL_Gold_EMT.M_MED_HighTower_Tomato_Body_Armor_SL_Gold_EMT" :Option%203 "/BRCosmetics/Characters/Player/Male/Medium/Bodies/M_MED_HighTower_Tomato/Materials/M_MED_HighTower_Tomato_Body_Armor_SL_EMT.M_MED_HighTower_Tomato_Body_Armor_SL_EMT") "Armor Body Masked" "None" :!class /Script/Engine.KismetMaterialLibrary))
            (LocalVariables|SetArmor%20Body%20Emote%20MID _returnvalue)
            (bind _returnvalue_2 (Rendering|Material|CreateDynamicMaterialInstance (Utilities|Select :Index (LocalVariables|GetCurrent%20Variant) :Option%200 "/BRCosmetics/Characters/Player/Male/Medium/Bodies/M_MED_HighTower_Tomato_Casual/Materials/M_MED_HighTower_TomatoCasual_Body_EMOTE.M_MED_HighTower_TomatoCasual_Body_EMOTE" :Option%201 "/BRCosmetics/Characters/Player/Male/Medium/Bodies/M_MED_HighTower_Tomato_Casual/Materials/M_MED_HighTower_TomatoCasual_Body_SL_Slvr_EMT.M_MED_HighTower_TomatoCasual_Body_SL_Slvr_EMT" :Option%202 "/BRCosmetics/Characters/Player/Male/Medium/Bodies/M_MED_HighTower_Tomato_Casual/Materials/M_MED_HighTower_TomatoCasual_Body_SL_Gold_EMT.M_MED_HighTower_TomatoCasual_Body_SL_Gold_EMT" :Option%203 "/BRCosmetics/Characters/Player/Male/Medium/Bodies/M_MED_HighTower_Tomato_Casual/Materials/M_MED_HighTower_TomatoCasual_Body_SL_Foil_EMT.M_MED_HighTower_TomatoCasual_Body_SL_Foil_EMT") "Casual Body Masked" "None" :!class /Script/Engine.KismetMaterialLibrary))
            (LocalVariables|SetCasual%20Body%20Emote%20MID _returnvalue_2)
            (bind _returnvalue_3 (Rendering|Material|CreateDynamicMaterialInstance (Utilities|Select :Index (LocalVariables|GetCurrent%20Variant) :Option%200 "/BRCosmetics/Characters/Player/Male/Medium/Bodies/M_MED_HighTower_Tomato_Casual/Materials/M_MED_HighTower_TomatoCasual_FaceAcc_EMOTE.M_MED_HighTower_TomatoCasual_FaceAcc_EMOTE" :Option%201 "/BRCosmetics/Characters/Player/Male/Medium/Bodies/M_MED_HighTower_Tomato_Casual/Materials/M_MED_HighTower_TomatoCasual_FaceAcc_SL_EMT.M_MED_HighTower_TomatoCasual_FaceAcc_SL_EMT" :Option%202 "/BRCosmetics/Characters/Player/Male/Medium/Bodies/M_MED_HighTower_Tomato_Casual/Materials/M_MED_HighTower_TomatoCasual_FaceAcc_SL_EMT.M_MED_HighTower_TomatoCasual_FaceAcc_SL_EMT" :Option%203 "/BRCosmetics/Characters/Player/Male/Medium/Bodies/M_MED_HighTower_Tomato_Casual/Materials/M_MED_HighTower_TomatoCasual_FaceAcc_SL_EMT.M_MED_HighTower_TomatoCasual_FaceAcc_SL_EMT") "Hair Masked " "None" :!class /Script/Engine.KismetMaterialLibrary))
            (LocalVariables|SetHairEmote_MID _returnvalue_3)
            (bind _returnvalue_4 (Rendering|Material|CreateDynamicMaterialInstance "/BRCosmetics/Characters/Player/Male/Medium/Bodies/M_MED_HighTower_Tomato_Casual/Materials/M_MED_Hightower_TomatoCasual_Glass.M_MED_Hightower_TomatoCasual_Glass" "Glasses" "None" :!class /Script/Engine.KismetMaterialLibrary))
            (LocalVariables|SetGlassEmote_MID _returnvalue_4)
            (bind _returnvalue_5 (Rendering|Material|CreateDynamicMaterialInstance (Utilities|Select :Index (LocalVariables|GetCurrent%20Variant) :Option%200 "/BRCosmetics/Characters/Player/Male/Medium/Bodies/M_MED_HighTower_Tomato_Casual/Materials/M_MED_HighTower_Tomato_FaceAcc_Armor_EMOTE.M_MED_HighTower_Tomato_FaceAcc_Armor_EMOTE" :Option%201 "/BRCosmetics/Characters/Player/Male/Medium/Bodies/M_MED_HighTower_Tomato/Materials/M_MED_HighTower_Tomato_FaceAcc_Armor_SL_Silver_EMT.M_MED_HighTower_Tomato_FaceAcc_Armor_SL_Silver_EMT" :Option%202 "/BRCosmetics/Characters/Player/Male/Medium/Bodies/M_MED_HighTower_Tomato/Materials/M_MED_HighTower_Tomato_FaceAcc_Armor_SL_Gold_EMT.M_MED_HighTower_Tomato_FaceAcc_Armor_SL_Gold_EMT" :Option%203 "/BRCosmetics/Characters/Player/Male/Medium/Bodies/M_MED_HighTower_Tomato/Materials/M_MED_HighTower_Tomato_FaceAcc_Armor_SL_Foil_EMT.M_MED_HighTower_Tomato_FaceAcc_Armor_SL_Foil_EMT") "Armor Head Masked" "None" :!class /Script/Engine.KismetMaterialLibrary))
            (LocalVariables|SetArmor%20Head%20Emote%20MID _returnvalue_5)
            (bind _returnvalue_6 (Rendering|Material|CreateDynamicMaterialInstance (Utilities|Select :Index (LocalVariables|GetCurrent%20Variant) :Option%200 "/BRCosmetics/Characters/Player/Male/Medium/Bodies/M_MED_HighTower_Tomato_Casual/Materials/M_MED_HighTower_TomatoCasual_Head_EMOTE.M_MED_HighTower_TomatoCasual_Head_EMOTE" :Option%201 "/BRCosmetics/Characters/Player/Male/Medium/Bodies/M_MED_HighTower_Tomato_Casual/Materials/M_MED_HighTower_TomatoCasual_Head_SL_Foil_EMT.M_MED_HighTower_TomatoCasual_Head_SL_Foil_EMT" :Option%202 "/BRCosmetics/Characters/Player/Male/Medium/Bodies/M_MED_HighTower_Tomato_Casual/Materials/M_MED_HighTower_TomatoCasual_Head_SL_Foil_EMT.M_MED_HighTower_TomatoCasual_Head_SL_Foil_EMT" :Option%203 "/BRCosmetics/Characters/Player/Male/Medium/Bodies/M_MED_HighTower_Tomato_Casual/Materials/M_MED_HighTower_TomatoCasual_Head_SL_Foil_EMT.M_MED_HighTower_TomatoCasual_Head_SL_Foil_EMT") "Casual Head Masked" "None" :!class /Script/Engine.KismetMaterialLibrary))
            (LocalVariables|SetCasual%20Head%20Emote%20MID _returnvalue_6))
          (:then_1
            (bind _asfort_player_pawn_2 (Utilities|Casting|CastToFortPlayerPawn (Components|GetOwner MeshComp)))
            (Rendering|Material|SetMaterial (Customization|GetSkeletalMeshforPartType _asfort_player_pawn_2 "Head" :!class /Script/FortniteGame.FortPlayerPawn) 0 (LocalVariables|GetCasual%20Head%20Emote%20MID))
            (Rendering|Material|SetMaterial (Customization|GetSkeletalMeshforPartType _asfort_player_pawn_2 "Head" :!class /Script/FortniteGame.FortPlayerPawn) 1 (LocalVariables|GetArmor%20Head%20Emote%20MID))
            (Rendering|Material|SetMaterial (Customization|GetSkeletalMeshforPartType _asfort_player_pawn_2 "Face" :!class /Script/FortniteGame.FortPlayerPawn) 0 (LocalVariables|GetHairEmote_MID))
            (Rendering|Material|SetMaterial (Customization|GetSkeletalMeshforPartType _asfort_player_pawn_2 "Face" :!class /Script/FortniteGame.FortPlayerPawn) 1 (LocalVariables|GetGlassEmote_MID))
            (Rendering|Material|SetMaterial (Customization|GetSkeletalMeshforPartType _asfort_player_pawn_2 "Body" :!class /Script/FortniteGame.FortPlayerPawn) 0 (LocalVariables|GetCasual%20Body%20Emote%20MID))
            (Rendering|Material|SetMaterial (Customization|GetSkeletalMeshforPartType _asfort_player_pawn_2 "Body" :!class /Script/FortniteGame.FortPlayerPawn) 1 (LocalVariables|GetArmor%20Body%20Emote%20MID))))
        (else
          (Fortnite|FortShippingLog 0 (Utilities|String|Append "Hightower Tomato Custom Notify -- " "AnimNotify_Tower_CasualtoMask" " : Had Progressive Active at Index [" (Utilities|String|ToString(Integer) (LocalVariables|GetCurrent%20Progressive)) "] but expected Progressive Active at Index [" (Utilities|String|ToString(Integer) (LocalVariables|GetExpected%20Progressive)) "]! Nothing to do, aborting!") true))))
    (:then_2
      (return false))))
