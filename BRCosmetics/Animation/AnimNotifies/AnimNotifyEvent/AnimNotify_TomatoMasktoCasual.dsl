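; Anim notify state that switches the HighTower Tomato skin from armored mask to casual materials during an animation.
; Begin pushes the active Progressive stage into a BroccoliEmoteSync data store as "BrocolliEmoteHack"; End looks it up under the differently spelled "BroccoliEmoteHack".
; If that stage matches the expected one, End builds dynamic materials for the Material variant, sets HT_CrunchVerts and applies them to Head, Face and Body; otherwise it logs.
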(fn Received_NotifyBegin
  (params (MeshComp :type Skeletal%20Mesh%20Component%20Object%20Reference) (Animation :type Anim%20Sequence%20Base%20Object%20Reference) (TotalDuration :type Float%20%28single-precision%29) (EventReference :type Anim%20Notify%20Event%20Reference%20Structure%20%28by%20ref%29))
  (Utilities|FlowControl|Sequence
    (:then_0
      (bind _asfort_player_pawn (Utilities|Casting|CastToFortPlayerPawn (Components|GetOwner MeshComp)))
      (bind _returnvalue (DataStore|AddNamedStorebyType (Actor|GetComponentbyClass _asfort_player_pawn "/Script/FortniteGame.FortGenericDataStoreManagerComponent") "BrocolliEmoteHack" "/BRCosmetics/Animation/AnimNotifies/AnimNotifyEvent/BroccoliHacks/FortDataStore_BroccoliEmoteSync.FortDataStore_BroccoliEmoteSync_C"))
      (bind _asfort_data_store_broccoli_emote_sync (Utilities|Casting|CastToFortDataStore_BroccoliEmoteSync _returnvalue))
      (bind _index (CosmeticLoadout|GetVariantChannelActiveIndexforSlotActor 0 "(TagName=\"LoadoutSlot.Character\")" _asfort_player_pawn "(TagName=\"Cosmetics.Variant.Channel.Progressive\")" (Utilities|Array|MakeArray "(TagName=\"Cosmetics.Variant.Property.Stage1\")" "(TagName=\"Cosmetics.Variant.Property.Stage2\")")))
      (Class|FortDataStoreBroccoliEmoteSync|PushNotifyStartChannel _asfort_data_store_broccoli_emote_sync self _index))
    (:then_1
      (return false))))

(fn Received_NotifyEnd
  (params (MeshComp :type Skeletal%20Mesh%20Component%20Object%20Reference) (Animation :type Anim%20Sequence%20Base%20Object%20Reference) (EventReference :type Anim%20Notify%20Event%20Reference%20Structure%20%28by%20ref%29))
  (Utilities|FlowControl|Sequence
    (:then_0
      (bind _asfort_player_pawn (Utilities|Casting|CastToFortPlayerPawn (Components|GetOwner MeshComp))
        (:then
          (bind _index (CosmeticLoadout|GetVariantChannelActiveIndexforSlotActor 0 "(TagName=\"LoadoutSlot.Character\")" _asfort_player_pawn "(TagName=\"Cosmetics.Variant.Channel.Material\")" (Utilities|Array|MakeArray "(TagName=\"Cosmetics.Variant.Property.Mat1\")" "(TagName=\"Cosmetics.Variant.Property.Mat2\")" "(TagName=\"Cosmetics.Variant.Property.Mat3\")" "(TagName=\"Cosmetics.Variant.Property.Mat4\")")))
          (LocalVariables|SetCurrent%20Variant _index)
          (bind _asfort_data_store_broccoli_emote_sync (Utilities|Casting|CastToFortDataStore_BroccoliEmoteSync (DataStore|GetNamedStorebyType (Actor|GetComponentbyClass _asfort_player_pawn "/Script/FortniteGame.FortGenericDataStoreManagerComponent") "BroccoliEmoteHack" "/BRCosmetics/Animation/AnimNotifies/AnimNotifyEvent/BroccoliHacks/FortDataStore_BroccoliEmoteSync.FortDataStore_BroccoliEmoteSync_C")))
          (bind _channelidx (Class|FortDataStoreBroccoliEmoteSync|GetNotifyStartChannel _asfort_data_store_broccoli_emote_sync self))
          (LocalVariables|SetCurrent%20Progressive _channelidx))
        (:CastFailed
          (Development|PrintString 0 "FortPlayerPawn CAST FAILED"))))
    (:then_1
      (if (== (LocalVariables|GetCurrent%20Progressive) (LocalVariables|GetExpected%20Progressive))
        (Utilities|FlowControl|Sequence
          (:then_0
            (bind _returnvalue (Rendering|Material|CreateDynamicMaterialInstance MeshComp (Utilities|Select :Index (LocalVariables|GetCurrent%20Variant) :Option%200 "/BRCosmetics/Characters/Player/Male/Medium/Bodies/M_MED_HighTower_Tomato/Materials/M_MED_HighTower_Tomato_Body_Armor_Opaque.M_MED_HighTower_Tomato_Body_Armor_Opaque" :Option%201 "/BRCosmetics/Characters/Player/Male/Medium/Bodies/M_MED_HighTower_Tomato/Materials/M_MED_HighTower_Tomato_Body_Armor_SL_Silver.M_MED_HighTower_Tomato_Body_Armor_SL_Silver" :Option%202 "/BRCosmetics/Characters/Player/Male/Medium/Bodies/M_MED_HighTower_Tomato/Materials/M_MED_HighTower_Tomato_Body_Armor_SL_Gold.M_MED_HighTower_Tomato_Body_Armor_SL_Gold" :Option%203 "/BRCosmetics/Characters/Player/Male/Medium/Bodies/M_MED_HighTower_Tomato/Materials/M_MED_HighTower_Tomato_Body_Armor_SL.M_MED_HighTower_Tomato_Body_Armor_SL") "Armor Body " "None" :!class /Script/Engine.KismetMaterialLibrary))
            (LocalVariables|SetArmor%20Body%20Opaque%20MID _returnvalue)
            (bind _returnvalue_2 (Rendering|Material|CreateDynamicMaterialInstance MeshComp (Utilities|Select :Index (LocalVariables|GetCurrent%20Variant) :Option%200 "/BRCosmetics/Characters/Player/Male/Medium/Bodies/M_MED_HighTower_Tomato_Casual/Materials/M_MED_HighTower_TomatoCasual_Body_Opaque.M_MED_HighTower_TomatoCasual_Body_Opaque" :Option%201 "/BRCosmetics/Characters/Player/Male/Medium/Bodies/M_MED_HighTower_Tomato_Casual/Materials/M_MED_HighTower_TomatoCasual_Body_SL_Silver.M_MED_HighTower_TomatoCasual_Body_SL_Silver" :Option%202 "/BRCosmetics/Characters/Player/Male/Medium/Bodies/M_MED_HighTower_Tomato_Casual/Materials/M_MED_HighTower_TomatoCasual_Body_SL_Gold.M_MED_HighTower_TomatoCasual_Body_SL_Gold" :Option%203 "/BRCosmetics/Characters/Player/Male/Medium/Bodies/M_MED_HighTower_Tomato_Casual/Materials/M_MED_HighTower_TomatoCasual_Body_SL_Foil.M_MED_HighTower_TomatoCasual_Body_SL_Foil") "Casual Body Masked" "None" :!class /Script/Engine.KismetMaterialLibrary))
            (LocalVariables|SetCasual%20Body%20Opaque%20MID _returnvalue_2)
            (bind _returnvalue_3 (Rendering|Material|CreateDynamicMaterialInstance MeshComp (Utilities|Select :Index (LocalVariables|GetCurrent%20Variant) :Option%200 "/BRCosmetics/Characters/Player/Male/Medium/Bodies/M_MED_HighTower_Tomato_Casual/Materials/M_MED_HighTower_TomatoCasual_FaceAcc_Opaque.M_MED_HighTower_TomatoCasual_FaceAcc_Opaque" :Option%201 "/BRCosmetics/Characters/Player/Male/Medium/Bodies/M_MED_HighTower_Tomato_Casual/Materials/M_MED_HighTower_TomatoCasual_FaceAcc_SL_Foil.M_MED_HighTower_TomatoCasual_FaceAcc_SL_Foil" :Option%202 "/BRCosmetics/Characters/Player/Male/Medium/Bodies/M_MED_HighTower_Tomato_Casual/Materials/M_MED_HighTower_TomatoCasual_FaceAcc_SL_Foil.M_MED_HighTower_TomatoCasual_FaceAcc_SL_Foil" :Option%203 "/BRCosmetics/Characters/Player/Male/Medium/Bodies/M_MED_HighTower_Tomato_Casual/Materials/M_MED_HighTower_TomatoCasual_FaceAcc_SL_Foil.M_MED_HighTower_TomatoCasual_FaceAcc_SL_Foil") "Hair Masked " "None" :!class /Script/Engine.KismetMaterialLibrary))
            (LocalVariables|SetHairOpaque_MID _returnvalue_3)
            (bind _returnvalue_4 (Rendering|Material|CreateDynamicMaterialInstance MeshComp (Utilities|Select :Index (LocalVariables|GetCurrent%20Variant) :Option%200 "/BRCosmetics/Characters/Player/Male/Medium/Bodies/M_MED_HighTower_Tomato/Materials/M_MED_HighTower_Tomato_FaceAcc_Armor_Opaque.M_MED_HighTower_Tomato_FaceAcc_Armor_Opaque" :Option%201 "/BRCosmetics/Characters/Player/Male/Medium/Bodies/M_MED_HighTower_Tomato/Materials/M_MED_HighTower_Tomato_FaceAcc_Armor_SL_Silver.M_MED_HighTower_Tomato_FaceAcc_Armor_SL_Silver" :Option%202 "/BRCosmetics/Characters/Player/Male/Medium/Bodies/M_MED_HighTower_Tomato/Materials/M_MED_HighTower_Tomato_FaceAcc_Armor_SL_Gold.M_MED_HighTower_Tomato_FaceAcc_Armor_SL_Gold" :Option%203 "/Game/Characters/Player/Male/Medium/Bodies/M_MED_HighTower_Tomato/Materials/M_MED_HighTower_Tomato_FaceAcc_Armor_SL_Foil.M_MED_HighTower_Tomato_FaceAcc_Armor_SL_Foil") "Armor Head Opaque" "None" :!class /Script/Engine.KismetMaterialLibrary))
            (LocalVariables|SetArmor%20Head%20Opaque%20MID _returnvalue_4)
            (bind _returnvalue_5 (Rendering|Material|CreateDynamicMaterialInstance MeshComp (Utilities|Select :Index (LocalVariables|GetCurrent%20Variant) :Option%200 "/BRCosmetics/Characters/Player/Male/Medium/Bodies/M_MED_HighTower_Tomato_Casual/Materials/M_MED_HighTower_TomatoCasual_Head_Opaque.M_MED_HighTower_TomatoCasual_Head_Opaque" :Option%201 "/BRCosmetics/Characters/Player/Male/Medium/Bodies/M_MED_HighTower_Tomato_Casual/Materials/M_MED_HighTower_TomatoCasual_Head_SL_Foil.M_MED_HighTower_TomatoCasual_Head_SL_Foil" :Option%202 "/BRCosmetics/Characters/Player/Male/Medium/Bodies/M_MED_HighTower_Tomato_Casual/Materials/M_MED_HighTower_TomatoCasual_Head_SL_Foil.M_MED_HighTower_TomatoCasual_Head_SL_Foil" :Option%203 "/BRCosmetics/Characters/Player/Male/Medium/Bodies/M_MED_HighTower_Tomato_Casual/Materials/M_MED_HighTower_TomatoCasual_Head_SL_Foil.M_MED_HighTower_TomatoCasual_Head_SL_Foil") "Casual Head Opaque" "None" :!class /Script/Engine.KismetMaterialLibrary))
            (LocalVariables|SetCasual%20Head%20Opaque%20MID _returnvalue_5))
          (:then_1
            (bind _asfort_player_pawn_2 (Utilities|Casting|CastToFortPlayerPawn (Components|GetOwner MeshComp)))
            (Rendering|Material|SetScalarParameterValue (LocalVariables|GetArmor%20Body%20Opaque%20MID) "HT_CrunchVerts" 1.0 :!class /Script/Engine.MaterialInstanceDynamic)
            (Rendering|Material|SetScalarParameterValue (LocalVariables|GetCasual%20Head%20Opaque%20MID) "HT_CrunchVerts" 0.0 :!class /Script/Engine.MaterialInstanceDynamic)
            (Rendering|Material|SetMaterial (Customization|GetSkeletalMeshforPartType _asfort_player_pawn_2 "Head" :!class /Script/FortniteGame.FortPlayerPawn) 0 (LocalVariables|GetCasual%20Head%20Opaque%20MID))
            (Rendering|Material|SetMaterial (Customization|GetSkeletalMeshforPartType _asfort_player_pawn_2 "Head" :!class /Script/FortniteGame.FortPlayerPawn) 1 (LocalVariables|GetArmor%20Head%20Opaque%20MID))
            (Rendering|Material|SetMaterial (Customization|GetSkeletalMeshforPartType _asfort_player_pawn_2 "Face" :!class /Script/FortniteGame.FortPlayerPawn) 0 (LocalVariables|GetHairOpaque_MID))
            (Rendering|Material|SetMaterial (Customization|GetSkeletalMeshforPartType _asfort_player_pawn_2 "Body" :!class /Script/FortniteGame.FortPlayerPawn) 0 (LocalVariables|GetCasual%20Body%20Opaque%20MID))
            (Rendering|Material|SetMaterial (Customization|GetSkeletalMeshforPartType _asfort_player_pawn_2 "Body" :!class /Script/FortniteGame.FortPlayerPawn) 1 (LocalVariables|GetArmor%20Body%20Opaque%20MID))))
        (else
          (Fortnite|FortShippingLog 0 (Utilities|String|Append "Hightower Tomato Custom Notify -- " "AnimNotify_Tower_MaskToCasual" " : Had Progressive Active at Index [" (Utilities|String|ToString(Integer) (LocalVariables|GetCurrent%20Progressive)) "] but expected Progressive Active at Index [" (Utilities|String|ToString(Integer) (LocalVariables|GetExpected%20Progressive)) "]! Nothing to do, aborting!")))))
    (:then_2
      (return false))))
