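; Anim notify state for the BoldDorm Bagel emote, apparently driving video-style playback in a material on the skeletal mesh.
; Each tick looks up the dynamic material instance at a configurable material index and checks that it is valid.
; If valid, it writes the current notify state time into a named scalar parameter on the MID variable, so animation time drives the material.
; NotifyBegin does no work.
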(fn Received_NotifyBegin
  (params (MeshComp :type Skeletal%20Mesh%20Component%20Object%20Reference) (Animation :type Anim%20Sequence%20Base%20Object%20Reference) (TotalDuration :type Float%20%28single-precision%29) (EventReference :type Anim%20Notify%20Event%20Reference%20Structure%20%28by%20ref%29))
  (return false))

(fn Received_NotifyTick
  (params (MeshComp :type Skeletal%20Mesh%20Component%20Object%20Reference) (Animation :type Anim%20Sequence%20Base%20Object%20Reference) (FrameDeltaTime :type Float%20%28single-precision%29) (EventReference :type Anim%20Notify%20Event%20Reference%20Structure%20%28by%20ref%29))
  (Utilities|IsValid (Class|BTechArtGlobalLibrary|GetDynamicMaterialInstance MeshComp (Variables|Default|GetMaterial%20Index) MeshComp)
    (:Is%20Valid
      (Rendering|Material|SetScalarParameterValue (Variables|Default|GetMID) (Variables|Default|GetParameterName) (Utilities|Animation|Notifies|GetCurrentAnimationNotifyStateTime EventReference) :!class /Script/Engine.MaterialInstanceDynamic)
      (return true))
    (:Is%20Not%20Valid
      (return false))))
