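; Anim notify state for the Ceremonial Guard cosmetic that feeds the player's kill count into a particle effect during an animation.
; Each tick it reads the StatManager.AthenaKills cosmetic stat from the mesh owner's FortPlayerPawn.
; It writes that value as a float to the NumKills parameter of the notify's spawned effect.
; A Debug flag selects between the real stat and a debug variable.
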(fn Received_NotifyTick
  (params (MeshComp :type Skeletal%20Mesh%20Component%20Object%20Reference) (Animation :type Anim%20Sequence%20Base%20Object%20Reference) (FrameDeltaTime :type Float%20%28single-precision%29) (EventReference :type Anim%20Notify%20Event%20Reference%20Structure%20%28by%20ref%29))
  (bind _returnvalue (AnimNotify|GetSpawnedEffect MeshComp))
  (bind _asfort_player_pawn (Utilities|Casting|CastToFortPlayerPawn (Components|GetOwner MeshComp)))
  (bind _returnvalue_2 (Class|FortCosmeticFunctionLibrary|GetCosmeticStatValue _asfort_player_pawn "(TagName=\"StatManager.AthenaKills\")"))
  (Effects|Components|ParticleSystem|SetFloatParameter _returnvalue "NumKills" (select (Variables|Default|GetDebug) (Math|Conversions|ToFloat(Integer) (Variables|Default|GetDebugVar)) (Math|Conversions|ToFloat(Integer) _returnvalue_2)))
  (return false))
