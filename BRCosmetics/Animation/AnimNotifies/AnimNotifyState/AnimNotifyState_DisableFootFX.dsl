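; Anim notify state that can switch off a Fortnite player pawn's movement dust FX for the span of an animation.
; On begin, it casts the mesh owner to FortPlayerPawn and clears bFXPlayDustOnMovement only if the FootFXType Disable Default Dust option is set.
; On end, it sets the flag back to true regardless of that option or its earlier value.
; Both events return false when the cast fails.
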(fn Received_NotifyBegin
  (params (MeshComp :type Skeletal%20Mesh%20Component%20Object%20Reference) (Animation :type Anim%20Sequence%20Base%20Object%20Reference) (TotalDuration :type Float%20%28single-precision%29) (EventReference :type Anim%20Notify%20Event%20Reference%20Structure%20%28by%20ref%29))
  (bind _asfort_player_pawn (Utilities|Casting|CastToFortPlayerPawn (Components|GetOwner MeshComp))
    (:then
      (if (Variables|FootFXType|GetDisable%20Default%20Dust)
        (Variables|/Script/FortniteGame.FortPlayerPawn|SetbFXPlayDustOnMovement false _asfort_player_pawn))
      (return true))
    (:CastFailed
      (return false))))

(fn Received_NotifyEnd
  (params (MeshComp :type Skeletal%20Mesh%20Component%20Object%20Reference) (Animation :type Anim%20Sequence%20Base%20Object%20Reference) (EventReference :type Anim%20Notify%20Event%20Reference%20Structure%20%28by%20ref%29))
  (bind _asfort_player_pawn (Utilities|Casting|CastToFortPlayerPawn (Components|GetOwner MeshComp))
    (:then
      (Variables|/Script/FortniteGame.FortPlayerPawn|SetbFXPlayDustOnMovement true _asfort_player_pawn)
      (return true))
    (:CastFailed
      (return false))))
