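; Emote anim notify state that drives a scalar material parameter on a chosen character part of a Fort player pawn, intended to hide an element during the notify window.
; Begin queues work until that part is applied, passing parameter name and begin value as strings.
; The MaterialVertHideElement callback unpacks these and sets the parameter on the part's mesh materials.
; End sets the end value immediately on the part's current skeletal mesh, without queuing.
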(fn MaterialVertHideElement
  (params (Pawn :type Fort%20Player%20Pawn%20Object%20Reference) (Part :type Custom%20Character%20Part%20Object%20Reference) (MeshPart :type Skeletal%20Mesh%20Component%20Object%20Reference) (PartType :type EFortCustomPartType%20Enum :default "Head") (AnyStruct :type Instanced%20Struct%20Structure%20%28by%20ref%29))
  (bind (_objectarray _stringarray) (Utilities|UnpackDatawhenCosmeticPartApplied AnyStruct)
    (:Valid
      (Rendering|Material|SetScalarParameterValueonMaterials MeshPart (Utilities|String|StringToName (Utilities|Array|Get(aref) _stringarray)) (Utilities|String|StringToFloat (Utilities|Array|Get(aref) _stringarray 1)))
      (return))))

(fn Received_NotifyBegin
  (params (MeshComp :type Skeletal%20Mesh%20Component%20Object%20Reference) (Animation :type Anim%20Sequence%20Base%20Object%20Reference) (TotalDuration :type Float%20%28single-precision%29) (EventReference :type Anim%20Notify%20Event%20Reference%20Structure%20%28by%20ref%29))
  ;[0.4172 0.7642 1] Get Active Character Part
  (bind _asfort_player_pawn (Utilities|Casting|CastToFortPlayerPawn (Components|GetOwner MeshComp)))
  ;[0.4549 1 0.7631] Material Vert Hide Element
  (Default|QueueWorkwhenSpecificCosmeticPartApplied :Target _asfort_player_pawn :Delegate (EventDispatchers|CreateEvent :!event MaterialVertHideElement) :PartType (Variables|Default|GetPart%20Type) :StringArray (Utilities|Array|MakeArray (Utilities|String|ToString(Name) (Variables|Default|GetParameter%20Name)) (Utilities|String|ToString(Float) (Variables|Default|GetParameter%20Value%20Begin))))
  (return false))

(fn Received_NotifyEnd
  (params (MeshComp :type Skeletal%20Mesh%20Component%20Object%20Reference) (Animation :type Anim%20Sequence%20Base%20Object%20Reference) (EventReference :type Anim%20Notify%20Event%20Reference%20Structure%20%28by%20ref%29))
  ;[0.4172 0.7642 1] Get Active Character Part
  (bind _asfort_player_pawn (Utilities|Casting|CastToFortPlayerPawn (Components|GetOwner MeshComp)))
  ;[0.4549 1 0.7631] Material Vert Hide Element
  (Rendering|Material|SetScalarParameterValueonMaterials (Customization|GetSkeletalMeshforPartType _asfort_player_pawn (Variables|Default|GetPart%20Type) :!class /Script/FortniteGame.FortPlayerPawn) (Variables|Default|GetParameter%20Name) (Variables|Default|GetParameter%20Value%20End))
  (return false))
