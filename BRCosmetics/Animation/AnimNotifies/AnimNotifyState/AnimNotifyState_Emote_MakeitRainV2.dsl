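; Anim notify state for the Make It Rain V2 emote that spawns its particle effect when the notify window begins.
; Notify begin attaches the P_Emote_MakeItRainV2 emitter to the mesh's FX_Head point, snapped to target and set to auto-destroy, then enables shadow casting on it.
; The file shows no tick or end handling, so nothing here explicitly stops the effect.
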(fn Received_NotifyBegin
  (params (MeshComp :type Skeletal%20Mesh%20Component%20Object%20Reference) (Animation :type Anim%20Sequence%20Base%20Object%20Reference) (TotalDuration :type Float%20%28single-precision%29) (EventReference :type Anim%20Notify%20Event%20Reference%20Structure%20%28by%20ref%29))
  (bind _returnvalue (Effects|SpawnEmitterAttached :EmitterTemplate "/BRCosmetics/Effects/Fort_Effects/Emotes/P_Emote_MakeItRainV2.P_Emote_MakeItRainV2" :AttachToComponent MeshComp :AttachPointName "FX_Head" :Scale "1.000000,1.000000,1.000000" :LocationType "SnapToTarget" :bAutoDestroy true :PoolingMethod "None" :bAutoActivate true))
  (Rendering|SetCastShadow _returnvalue true)
  (return false))
