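; Anim notify state that plays the coin visual effect for the Toss Coin emote.
; When the notify begins, it spawns the NS_Emote_TossCoin_V3 Niagara system attached to the mesh's left hand with a small offset.
; It then sets the User.CoinSide float on the spawned component to a fixed value rather than a random one.
; Only the begin event is implemented, with no tick or end handling.
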(fn Received_NotifyBegin
  (params (MeshComp :type Skeletal%20Mesh%20Component%20Object%20Reference) (Animation :type Anim%20Sequence%20Base%20Object%20Reference) (TotalDuration :type Float%20%28single-precision%29) (EventReference :type Anim%20Notify%20Event%20Reference%20Structure%20%28by%20ref%29))
  (bind _returnvalue (Niagara|SpawnSystemAttached "/BRCosmetics/Athena/Effects/Particles/Emotes/NS_Emote_TossCoin/NS_Emote_TossCoin_V3.NS_Emote_TossCoin_V3" MeshComp "hand_l" "5.000000, 0, 0"))
  (Effects|Components|ParticleSystem|SetFloatParameter _returnvalue "User.CoinSide" 1.0)
  (return false))
