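; Anim notify state that configures the spawned Niagara effect for the Knight Cat emote.
; On begin, it reads the owner pawn's active pickaxe particle variant index and passes it to User.IndexColors.
; When that index is valid and the cosmetic is displayed in a frontend context, it also sets User.bIsFrontEndPreview.
; Debug strings print for the frontend branches and a failed variant check.
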(fn Received_NotifyBegin
  (params (MeshComp :type Skeletal%20Mesh%20Component%20Object%20Reference) (Animation :type Anim%20Sequence%20Base%20Object%20Reference) (TotalDuration :type Float%20%28single-precision%29) (EventReference :type Anim%20Notify%20Event%20Reference%20Structure%20%28by%20ref%29))
  (bind _returnvalue (AnimNotify|GetSpawnedEffect MeshComp))
  (bind _asniagara_particle_system_component (Utilities|Casting|CastToNiagaraComponent _returnvalue)
    (:then
      (LocalVariables|SetNiagara%20System _asniagara_particle_system_component)
      (bind _asfort_player_pawn (Utilities|Casting|CastToFortPlayerPawn (Components|GetOwner MeshComp)))
      (bind (_index _bindexisvalid) (CosmeticLoadout|GetVariantChannelActiveIndexforSlotActor 0 "(TagName=\"LoadoutSlot.Pickaxe\")" _asfort_player_pawn "(TagName=\"Cosmetics.Variant.Channel.Particle\")" (Utilities|Array|MakeArray "(TagName=\"Cosmetics.Variant.Property.Particle1\")" "(TagName=\"Cosmetics.Variant.Property.Particle2\")" "(TagName=\"Cosmetics.Variant.Property.Particle3\")")))
      (if _bindexisvalid
        (Niagara|SetNiagaraVariable(Int32) (LocalVariables|GetNiagara%20System) "User.IndexColors" _index)
        (bind _output_get (LocalVariables|SetblsFrontEnd (Reactive|Frontend|IsCosmeticObjectDisplayedinFrontendContext (Variables|GetMeshComp))))
        (if _output_get
          (Development|PrintString 0 "Setting Front End Variable" true true "(R=1.000000,G=0.000000,B=0.083156,A=1.000000)" 1e+01)
          (Niagara|SetNiagaraVariable(Bool) (LocalVariables|GetNiagara%20System) "User.bIsFrontEndPreview" (LocalVariables|GetblsFrontEnd))
          (else
            (Development|PrintString 0 "Not setting Front End Variable" true true "(R=1.000000,G=0.000000,B=0.083156,A=1.000000)" 1e+01)))
        (else
          (Development|PrintString 0 "Knight Cat Pickaxe Variant Slot Check FAILED")))
      (return true))
    (:CastFailed
      (return false))))
