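; Anim notify state for the Rose Depth emote effect, linking its spawned Niagara system to the player's body mesh.
; On each notify tick it casts the spawned effect to a Niagara component and the mesh owner to FortPlayerPawn.
; It then feeds the pawn's Body part skeletal mesh into the "PlayerPawn" Niagara parameter, returning false if either cast fails.
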(fn Received_NotifyTick
  (params (MeshComp :type Skeletal%20Mesh%20Component%20Object%20Reference) (Animation :type Anim%20Sequence%20Base%20Object%20Reference) (FrameDeltaTime :type Float%20%28single-precision%29) (EventReference :type Anim%20Notify%20Event%20Reference%20Structure%20%28by%20ref%29))
  (bind _returnvalue (AnimNotify|GetSpawnedEffect MeshComp))
  (bind _asniagara_particle_system_component (Utilities|Casting|CastToNiagaraComponent _returnvalue)
    (:then
      (LocalVariables|SetSpawnedNiagara _asniagara_particle_system_component)
      (bind _asfort_player_pawn (Utilities|Casting|CastToFortPlayerPawn (Components|GetOwner MeshComp))
        (:then
          (Niagara|SetNiagaraSkeletalMeshComponent (LocalVariables|GetSpawnedNiagara) "PlayerPawn" (Customization|GetSkeletalMeshforPartType _asfort_player_pawn "Body" :!class /Script/FortniteGame.FortPlayerPawn))
          (return true))
        (:CastFailed
          (return false))))
    (:CastFailed
      (return false))))
