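; Anim notify state for the SilentTempo emote that tunes a particle effect during animation.
; On notify begin it calls the parent implementation, then fetches the spawned effect for the mesh component if that call returns true.
; A valid effect receives the configured translucent sort priority and has its "DynamicFloat" particle parameter set from the notify's DynamicFloat variable.
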(fn Received_NotifyBegin
  (params (MeshComp :type Skeletal%20Mesh%20Component%20Object%20Reference) (Animation :type Anim%20Sequence%20Base%20Object%20Reference) (TotalDuration :type Float%20%28single-precision%29) (EventReference :type Anim%20Notify%20Event%20Reference%20Structure%20%28by%20ref%29))
  (bind _returnvalue (|CallParentFunction MeshComp Animation TotalDuration))
  (if _returnvalue
    (bind _returnvalue_2 (AnimNotify|GetSpawnedEffect MeshComp))
    (Utilities|IsValid _returnvalue_2
      (:Is%20Valid
        (Rendering|SetTranslucentSortPriority _returnvalue_2 (Variables|Default|GetTranslucent_Sort_Priority))
        (Effects|Components|ParticleSystem|SetFloatParameter _returnvalue_2 "DynamicFloat" (Variables|Default|GetDynamicFloat)))
      (:Is%20Not%20Valid
        _))
    (else
      _))
  (return _returnvalue))
