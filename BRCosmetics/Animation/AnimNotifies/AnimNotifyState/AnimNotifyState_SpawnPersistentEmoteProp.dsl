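; An anim notify state that spawns and manages a persistent emote prop on the pawn playing the animation.
; On begin, it has the owner FortPawn's emote component spawn the configured prop Blueprint at an attach socket, with optional scale inheritance.
; On end, it fetches the active prop and, if the notify ran to completion, detaches it keeping world transform and calls OnDetach; an interrupted notify destroys the prop instead.
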(fn Received_NotifyBegin
  (params (MeshComp :type Skeletal%20Mesh%20Component%20Object%20Reference) (Animation :type Anim%20Sequence%20Base%20Object%20Reference) (TotalDuration :type Float%20%28single-precision%29) (EventReference :type Anim%20Notify%20Event%20Reference%20Structure%20%28by%20ref%29))
  (bind _asfort_pawn (Utilities|Casting|CastToFortPawn (Components|GetOwner MeshComp)))
  (Class|FortEmoteComponent|SpawnPersistentProp (Emote|GetEmoteComponent _asfort_pawn) (Variables|Default|GetPersistent%20Prop%20BP) (Variables|Default|GetAttach%20Socket) (Variables|Default|GetInheritScale))
  (return false))

(fn Received_NotifyEnd
  (params (MeshComp :type Skeletal%20Mesh%20Component%20Object%20Reference) (Animation :type Anim%20Sequence%20Base%20Object%20Reference) (EventReference :type Anim%20Notify%20Event%20Reference%20Structure%20%28by%20ref%29))
  (bind _asfort_pawn (Utilities|Casting|CastToFortPawn (Components|GetOwner MeshComp)))
  (bind _asb_persistent_prop_base (Utilities|Casting|CastToB_PersistentProp_Base (PersistentProp|GetActivePersistentProp (Emote|GetEmoteComponent _asfort_pawn))))
  (LocalVariables|SetPersistent%20Prop _asb_persistent_prop_base)
  (if (Utilities|Animation|Notifies|NotifyStateReachedEnd EventReference)
    (Transformation|DetachFromActor (LocalVariables|GetPersistent%20Prop) "KeepWorld" "KeepWorld" "KeepWorld")
    (Class|FortPersistentProp|OnDetach (LocalVariables|GetPersistent%20Prop))
    (else
      (Class|FortPersistentProp|DestroyProp (LocalVariables|GetPersistent%20Prop) true)))
  (return false))
