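; Spawn-prop anim notify state for toon-styled cosmetics that corrects the prop mesh's look in frontend display contexts.
; CustomizePropMeshComponent checks whether the owning FortPawn is shown in the frontend, then loops over the prop's materials, reusing or creating dynamic material instances.
; Materials without PBR_Shading at 1 receive the frontend inverse LUT volume texture and matching LUT multiplier, mobile brightness and resolution parameters.
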(fn CustomizePropMeshComponent
  (params (MeshComponent :type Mesh%20Component%20Object%20Reference))
  (bind _currentmid (LocalVariables|GetCurrentMID))
  (LocalVariables|SetPropMesh MeshComponent)
  (bind _asfort_pawn (Utilities|Casting|CastToFortPawn (Components|GetOwner MeshComponent)))
  (if (Reactive|Frontend|IsCosmeticObjectDisplayedinFrontendContext _asfort_pawn)
    (for (_array_element _array_index) (Rendering|Material|GetMaterials (LocalVariables|GetPropMesh))
      (bind _asmaterial_instance_dynamic (Utilities|Casting|CastToMaterialInstanceDynamic _array_element)
        (:then
          (LocalVariables|SetCurrentMID _asmaterial_instance_dynamic))
        (:CastFailed
          (bind _returnvalue (Rendering|Material|CreateDynamicMaterialInstance _asfort_pawn _array_element "None" "None" :!class /Script/Engine.KismetMaterialLibrary))
          (bind _output_get (LocalVariables|SetCurrentMID _returnvalue))
          (Rendering|Material|SetMaterial (LocalVariables|GetPropMesh) _array_index _output_get)))
      (bind _returnvalue_2 (Rendering|Material|GetScalarParameterValue _currentmid "PBR_Shading" :!class /Script/Engine.MaterialInstanceDynamic))
      (if (not (== _returnvalue_2 1.0))
        (Rendering|Material|SetTextureParameterValue _currentmid "LUT Volume Texture" "/Game/Packages/Fortress_SharedMaterials/Anime_Master_Material/VT_InverseLUT_FE.VT_InverseLUT_FE")
        (Rendering|Material|SetScalarParameterValue _currentmid "LUT Multiplier" 1.0 :!class /Script/Engine.MaterialInstanceDynamic)
        (Rendering|Material|SetScalarParameterValue _currentmid "Mobile LUT Brightness" 1.0 :!class /Script/Engine.MaterialInstanceDynamic)
        (Rendering|Material|SetScalarParameterValue _currentmid "use 33 Resolution" 0.0 :!class /Script/Engine.MaterialInstanceDynamic)))))
