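; Anim notify state that customizes its prop mesh with a material matching the player's competitive rank.
; CustomizePropMeshComponent reads the rank cosmetic stat from the owning FortPlayerPawn and prints it as debug text.
; In editor or PIE, a DebugRank on an attached ranked pickaxe Blueprint overrides that stat for previewing.
; SwapRankedMaterial maps the rank integer to Bronze through Unreal tier materials and sets skeletal mesh slot 0.
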(fn CustomizePropMeshComponent
  (params (MeshComponent :type Mesh%20Component%20Object%20Reference))
  ; Query rank stat from FortPlayerPawn
  (bind _asfort_player_pawn (Utilities|Casting|CastToFortPlayerPawn (Components|GetOwner (Variables|GetMeshComponent))))
  (bind _returnvalue (Class|FortCosmeticFunctionLibrary|GetCosmeticStatValue _asfort_player_pawn (Variables|Rank|GetRankTag)))
  (bind _output_get (LocalVariables|SetRankTagValue _returnvalue))
  (Development|PrintText 0 (Utilities|Text|FormatText "NSLOCTEXT(\"[A5AD46B8D26C0E75F6B2A196A202D24E]\", \"2FD61EF54279311BCA51D597F35F3995\", \"Habanero Rank: {rank}\")" _output_get))
  (if (Editor|InEditororPIE (Variables|GetMeshComponent))
    ; If we're in PIE, will try to fetch the pickaxe BP to query the current rank. This way we can preview the rank using the ke * SetRank command like the other cosmetics.
    (bind _asfort_player_pawn_2 (Utilities|Casting|CastToFortPlayerPawn (Components|GetOwner (Variables|GetMeshComponent))))
    (bind _array_element (Utilities|Array|ForEachLoopwithBreak (Components|GetChildrenComponents (Variables|/Script/Engine.Character|GetMesh _asfort_player_pawn_2) false))
      (:LoopBody
        (bind _asb_athena_pickaxe_generic_ranked (Utilities|Casting|CastToB_Athena_Pickaxe_Generic_Ranked (Components|GetOwner _array_element)))
        (if (!= (Variables|/BRCosmetics/Athena/Cosmetics/Blueprints/B_Athena_Pickaxe_Generic_Ranked.B_Athena_Pickaxe_Generic_Ranked_C|GetDebugRank _asb_athena_pickaxe_generic_ranked) -1)
          (CallFunction|SwapRankedMaterial (Variables|GetMeshComponent) (Variables|/BRCosmetics/Athena/Cosmetics/Blueprints/B_Athena_Pickaxe_Generic_Ranked.B_Athena_Pickaxe_Generic_Ranked_C|GetDebugRank _asb_athena_pickaxe_generic_ranked))
          (LocalVariables|SetFoundRankedCosmetic true)))
      (:Completed
        _))
    (if (not (LocalVariables|GetFoundRankedCosmetic))
      (CallFunction|SwapRankedMaterial (Variables|GetMeshComponent) (LocalVariables|GetRankTagValue)))
    (else
      (CallFunction|SwapRankedMaterial (Variables|GetMeshComponent) (LocalVariables|GetRankTagValue)))))

(fn SwapRankedMaterial
  (params (MeshComponent :type Object%20Reference) (RankValue :type Integer))
  (bind _asskeletal_mesh_component (Utilities|Casting|CastToSkeletalMeshComponent MeshComponent))
  (Rendering|Material|SetMaterial _asskeletal_mesh_component 0 (Utilities|Select :Index (Variables|GetRankValue) :Option%200 (Variables|Rank|GetBronzeMaterial) :Option%201 (Variables|Rank|GetBronzeMaterial) :Option%202 (Variables|Rank|GetBronzeMaterial) :Option%203 (Variables|Rank|GetBronzeMaterial) :Option%204 (Variables|Rank|GetSilverMaterial) :Option%205 (Variables|Rank|GetSilverMaterial) :Option%206 (Variables|Rank|GetSilverMaterial) :Option%207 (Variables|Rank|GetGoldMaterial) :Option%208 (Variables|Rank|GetGoldMaterial) :Option%209 (Variables|Rank|GetGoldMaterial) :Option%2010 (Variables|Rank|GetPlatinumMaterial) :Option%2011 (Variables|Rank|GetPlatinumMaterial) :Option%2012 (Variables|Rank|GetPlatinumMaterial) :Option%2013 (Variables|Rank|GetDiamondMaterial) :Option%2014 (Variables|Rank|GetDiamondMaterial) :Option%2015 (Variables|Rank|GetDiamondMaterial) :Option%2016 (Variables|Rank|GetEliteMaterial) :Option%2017 (Variables|Rank|GetEliteMaterial) :Option%2018 (Variables|Rank|GetEliteMaterial) :Option%2019 (Variables|Rank|GetChampionMaterial) :Option%2020 (Variables|Rank|GetChampionMaterial) :Option%2021 (Variables|Rank|GetChampionMaterial) :Option%2022 (Variables|Rank|GetUnrealMaterial))))
