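; Anim notify state that swaps a Fortnite player pawn's cosmetic, evidently for skeleton-swap transformations, over the span of a montage.
; On begin, it prints an error if the state is too short and, when a transformation montage is set, registers it on the pawn with the current montage and start time.
; On end, a completed state disables the swap ID and listed GUIDs, then applies the swap data; a cancelled one clears the transformation montage instead.
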(fn Received_NotifyBegin
  (params (MeshComp :type Skeletal%20Mesh%20Component%20Object%20Reference) (Animation :type Anim%20Sequence%20Base%20Object%20Reference) (TotalDuration :type Float%20%28single-precision%29) (EventReference :type Anim%20Notify%20Event%20Reference%20Structure%20%28by%20ref%29))
  (Utilities|FlowControl|Sequence
    (:then_0
      (if (<= TotalDuration 0.5)
        (Development|PrintString 0 (LocalVariables|GetStateTooShortError) true true "(R=0.885417,G=0.005831,B=0.000000,A=1.000000)" 1e+01)))
    (:then_1
      (if (Utilities|IsValid (Variables|Default|GetTransformation%20Montage) :!class /Script/Engine.KismetSystemLibrary)
        (Utilities|FlowControl|Sequence
          (:then_0
            (bind _asfort_player_pawn (Utilities|Casting|CastToFortPlayerPawn (Components|GetOwner MeshComp))))
          (:then_1
            (bind _asanim_montage (Utilities|Casting|CastToAnimMontage Animation)))
          (:then_2
            (Class|FortPlayerPawn|SetTransformationMontage _asfort_player_pawn _asanim_montage (Variables|Default|GetTransformation%20Montage) (Variables|Default|GetStartTime) :!class /Script/FortniteGame.FortPlayerPawn)
            (return false)))
        (else
          (return false))))))

(fn Received_NotifyEnd
  (params (MeshComp :type Skeletal%20Mesh%20Component%20Object%20Reference) (Animation :type Anim%20Sequence%20Base%20Object%20Reference) (EventReference :type Anim%20Notify%20Event%20Reference%20Structure%20%28by%20ref%29))
  (bind _asfort_player_pawn (Utilities|Casting|CastToFortPlayerPawn (Components|GetOwner MeshComp)))
  (LocalVariables|SetPlayerPawn _asfort_player_pawn)
  ; Don't apply the cosmetic swap if this state is cancelled
  (if (Utilities|Animation|Notifies|NotifyStateReachedEnd (Variables|GetEventReference))
    (Utilities|FlowControl|Sequence
      (:then_0
        (Class|FortPlayerPawn|DisableCosmeticSwap (LocalVariables|GetPlayerPawn) (Variables|Default|GetSwap%20ID))
        (for _array_element (Variables|Default|GetClearGUIDs)
          (Class|FortPlayerPawn|DisableCosmeticSwap (LocalVariables|GetPlayerPawn) _array_element)))
      (:then_1
        (Class|FortPlayerPawn|SwapPawnCosmeticItemandVariant (LocalVariables|GetPlayerPawn) (Variables|Default|GetSwap%20ID) (Variables|Default|GetSwap%20Data) true "Active" false)
        (return false)))
    (else
      (Class|FortPlayerPawn|SetTransformationMontage (LocalVariables|GetPlayerPawn) 0 0 0.0 :!class /Script/FortniteGame.FortPlayerPawn))))
