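# Niagara one-shot dust puff, named for the Photographer Jenga cosmetic, built from a single Smoke_plum sprite emitter.
# An instantaneous burst spawns grey, low-alpha, two-second sprites just above the system location with upward and radial velocity.
# Drag and slight gravity slow the sprites while curves fade alpha, modulate size and advance through the later part of a 64-frame SubUV flipbook.
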NiagaraSystem /BRCosmetics/CT_NS_Photographer_JengaDust
dependencies {
 usedDataInterfaces [&NiagaraDataInterfaceCurve]
 usedDynamicInputs []
 usedModules [
  &self:SystemSpawnScript.SystemScriptSource.SystemScriptGraph.NiagaraNodeAssignment_0.SetVariables_#1
  &SystemState
 ]
}
emitters {Smoke_plum {
  emitterUpdateScript [
   &EmitterState()
   &SpawnBurst_Instantaneous(Spawn Count=24, Use Spawn Probability=false)
  ]
  particleSpawnScript [
   /Niagara/Modules/Spawn/Initialization/InitializeParticle(
    Color=&MakeLinearColor(
     R:NiagaraFloat=0.5
     G:NiagaraFloat=0.5
     B:NiagaraFloat=0.5
     A:NiagaraFloat=&UniformRangedFloat(Minimum=0.1, Maximum=0.15)
    )
    Lifetime=2
    Module.Write Color=true
    Module.Write Lifetime=true
    Module.Write Mass=true
    Module.Write Material Random=true
    Module.Write Position=false
    Module.Write Scale=false
    Module.Write SpriteRotation=true
    Module.Write SpriteSize=true
    Position=&SimulationPosition()
    Sprite Rotation=&UniformRangedFloat(Minimum=50, Maximum=100)
    Sprite Size=&Vector2DFromFloat(Value=&UniformRangedFloat(Minimum=45, Maximum=80))
   )
   &AddVelocity(Velocity=&UniformRangedVector(Minimum=<0 0 100>, Maximum=<0 0 500>))
   &AddVelocityFromPoint(
    Module.Use Falloff Distance=false
    Velocity Origin:Vector3f=&SimulationPosition()
    Velocity Strength=&UniformAOrBFloat(
     A=&UniformRangedFloat(Minimum=50, Maximum=150)
     B=&UniformRangedFloat(Minimum=50, Maximum=150)
    )
   )
   &SystemLocation(MaskSpawnGroup=false, Offset=<0 0 30>)
  ]
  particleUpdateScript [
   &ParticleState()
   &Drag(Drag=&UniformRangedFloat(Minimum=2, Maximum=4), Use Linear Drag=true, Use Rotational Drag=false)
   &GravityForce(Gravity=<0 0 -10>)
   &SolveForcesAndVelocity(Clamp Velocity=false, Limit Acceleration=false)
   /Niagara/Modules/Update/SubUV/SubUVAnimation(
    Number Of Frames=64
    SubUV Lookup Index=&FloatFromCurve(FloatCurve=DataInterface() {propertyValues {Curve {keys [
         {arriveTangent=0.3268, interpMode=RCIM_Cubic, leaveTangent=0.3268, tangentMode=RCTM_User, time=0, value=0.6}
         {arriveTangent=0.0289, interpMode=RCIM_Cubic, leaveTangent=0.0289, tangentMode=RCTM_User, time=1.0011, value=0.985}
        ]}}})
   )
   &ScaleColor(
    Scale Alpha=&FloatFromCurve(FloatCurve=DataInterface() {propertyValues {Curve {keys [
         {arriveTangent=-0.551, interpMode=RCIM_Cubic, leaveTangent=-0.551, tangentMode=RCTM_User, time=0, value=1}
         {arriveTangent=-2.3347, interpMode=RCIM_Cubic, leaveTangent=-2.3347, tangentMode=RCTM_User, time=1.0011, value=0.002}
        ]}}})
    ScaleA=true
    ScaleRGB=true
   )
   &ScaleSpriteSize(Scale Factor=&Vector2DFromFloat(Value=&FloatFromCurve(FloatCurve=DataInterface() {propertyValues {Curve {keys [<at -0.0046 1.0113 RCIM_Cubic>, {arriveTangent=1.728, interpMode=RCIM_Cubic, leaveTangent=1.728, tangentMode=RCTM_User, time=1, value=1}]}}})))
  ]
  rendererClasses [&NiagaraSpriteRendererProperties]
  renderers [{rendererClass=&NiagaraSpriteRendererProperties}]
 }}
summary {
 systemName=CT_NS_Photographer_JengaDust
 emitters [{emitterName=Smoke_plum, rendererClasses [&NiagaraSpriteRendererProperties]}]
 userVariables [
  User.GlobalOffset {type=Vector3f, description="", defaultValue {struct=Vector3f, value=<0 0 0>}}
  User.RibbonIDOffset {type=NiagaraInt32}
 ]
}
