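; Timed Niagara anim notify that binds its spawned effect to one body part's skeletal mesh on a Fortnite player pawn.
; On notify begin, it gets the spawned effect and casts the mesh owner to FortPlayerPawn and the effect to a Niagara component.
; It then feeds the mesh for the configured Part Type into the named Niagara skeletal mesh user parameter.
; Failed casts do nothing.
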(fn Received_NotifyBegin
  (params (MeshComp :type Skeletal%20Mesh%20Component%20Object%20Reference) (Animation :type Anim%20Sequence%20Base%20Object%20Reference) (TotalDuration :type Float%20%28single-precision%29) (EventReference :type Anim%20Notify%20Event%20Reference%20Structure%20%28by%20ref%29))
  (bind _returnvalue (AnimNotify|GetSpawnedEffect MeshComp))
  (bind _asfort_player_pawn (Utilities|Casting|CastToFortPlayerPawn (Components|GetOwner MeshComp))
    (:then
      (bind _asniagara_particle_system_component (Utilities|Casting|CastToNiagaraComponent _returnvalue)
        (:then
          (Niagara|SetNiagaraSkeletalMeshComponent _asniagara_particle_system_component (Variables|Default|GetSkeletal%20Mesh%20User%20Name) (Customization|GetSkeletalMeshforPartType _asfort_player_pawn (Variables|Default|GetPart%20Type) :!class /Script/FortniteGame.FortPlayerPawn)))
        (:CastFailed
          _)))
    (:CastFailed
      _))
  (return false))
